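# Control Rig for the F_MED_HayBrush female player body that applies procedural corrective deformation to sleeve, skirt and forearm twist bones.
# Spherical pose readers and a reusable swing/twist delta-rotation function turn limb poses into weights for additive bone offsets.
# Skirt bones follow the thigh twists, with correctives shaped by curves, anim attributes and state variables like crouch, slide, zipline and DBNO.
# A debug branch drives bones from controls and counter-rotates limb twists.
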ControlRigBlueprint /BRCosmetics/Characters/Player/Female/Medium/Bodies/F_MED_HayBrush/Meshes/F_MED_HayBrush_CtrlRig
graphs [
 Multiply_3_1_ContainedGraph {nodes [
   n1 = RigVMFunctionEntryNode()
   RigVMFunctionReturnNode(Result=&RigVMFunction_MathDoubleMul(B=@n1.C, A=&RigVMFunction_MathDoubleMul(A=@n1.A, B=@n1.B).Result).Result)
  ]}
 Multiply_5_1_ContainedGraph {nodes [
   n1 = RigVMFunctionEntryNode()
   RigVMFunctionReturnNode(Result=&RigVMFunction_MathDoubleMul(B=@n1.C, A=&RigVMFunction_MathDoubleMul(A=@n1.A, B=@n1.B).Result).Result)
  ]}
 New Function_ContainedGraph {nodes [
   n1 = RigVMFunctionEntryNode()
   RigVMFunction_ControlFlowBranch = &RigVMFunction_ControlFlowBranch(ExecuteContext=@n1.ExecuteContext, Condition=@n1.debug)
   RigVMFunctionReturnNode(ExecuteContext=@RigVMFunction_ControlFlowBranch.Completed)
   RigVMFunction_Sequence = RigVMAggregateNode(ExecuteContext=@RigVMFunction_ControlFlowBranch.True)
   RigUnit_ItemArray = &RigUnit_ItemArray(Items="((Type=Bone,Name=\"pelvis\"),(Type=Bone,Name=\"spine_01\"),(Type=Bone,Name=\"spine_02\"),(Type=Bone,Name=\"spine_03\"),(Type=Bone,Name=\"spine_04\"),(Type=Bone,Name=\"spine_05\"),(Type=Bone,Name=\"clavicle_l\"),(Type=Bone,Name=\"upperarm_l\"),(Type=Bone,Name=\"lowerarm_l\") ...(truncated)")
   DISPATCH_RigVMDispatch_ArrayIterator = RigVMDispatchNode(Array=@RigUnit_ItemArray.Items, ExecuteContext=@RigVMFunction_Sequence.A) {template="DISPATCH_RigVMDispatch_ArrayIterator(in Array,out Element,out Index,out Count,out Ratio)"}
   Concat = /Script/RigVM.RigVMFunction_NameConcat(B=_ctrl, A=@DISPATCH_RigVMDispatch_ArrayIterator.Element.Name) {is=struct RigVMUnitNode, template="Concat::Execute(in A,in B,out Result)"}
   > &RigUnit_SetTransform(
    Item=(Type=Bone,Name="spine_01")
    Weight=1
    bInitial=False
    Item.Name=@DISPATCH_RigVMDispatch_ArrayIterator.Element.Name
    Value=&RigUnit_GetTransform(Item=(Type=Control,Name="spine_01_ctrl"), bInitial=False, Item.Name=@Concat.Result).Transform
   )
   RigUnit_GetRelativeTransformForItem = &RigUnit_GetRelativeTransformForItem(Child=(Type=Bone,Name="thigh_l"), Parent=(Type=Null,Name="thigh_l_NULL"), bChildInitial=False, bParentInitial=False)
   RigVMFunction_MathQuaternionSwingTwist = &RigVMFunction_MathQuaternionSwingTwist(TwistAxis=(X=1,Y=0,Z=0), Input=@RigUnit_GetRelativeTransformForItem.RelativeTransform.Rotation)
   Inverse = /Script/RigVM.RigVMFunction_MathQuaternionInverse(Value=@RigVMFunction_MathQuaternionSwingTwist.Twist) {is=struct RigVMUnitNode, template="Inverse::Execute(in Value,out Result)"}
   &RigUnit_ModifyTransforms(
    ItemToModify=((Item=(Type=Bone,Name="thigh_twist_01_l"),Transform=(Rotation=(X=0,Y=0,Z=0,W=1),Translation=(X=0,Y=0,Z=0),Scale3D=(X=1,Y=1,Z=1))))
    Weight=1
    WeightMinimum=0
    ItemToModify.0.Transform.Rotation=@Inverse.Result
    ExecutePin=@RigVMFunction_Sequence.B
   )
   n2 = &RigUnit_GetRelativeTransformForItem(Child=(Type=Bone,Name="thigh_r"), Parent=(Type=Null,Name="thigh_r_null"), bChildInitial=False, bParentInitial=False)
   > &RigUnit_ModifyTransforms(
    ItemToModify=((Item=(Type=Bone,Name="thigh_twist_01_r"),Transform=(Rotation=(X=0,Y=0,Z=0,W=1),Translation=(X=0,Y=0,Z=0),Scale3D=(X=1,Y=1,Z=1))))
    Weight=1
    WeightMinimum=0
    ItemToModify.0.Transform.Rotation=&RigVMFunction_MathQuaternionInverse(Value=&RigVMFunction_MathQuaternionSwingTwist(TwistAxis=(X=1,Y=0,Z=0), Input=@n2.RelativeTransform.Rotation).Twist) {template="Inverse::Execute(in Value,out Result)"}.Result
   )
   n3 = &RigVMFunction_MathQuaternionInverse(Value=&RigVMFunction_MathQuaternionSwingTwist(
     TwistAxis=(X=1,Y=0,Z=0)
     Input=&RigUnit_GetRelativeTransformForItem(Child=(Type=Bone,Name="upperarm_l"), Parent=(Type=Null,Name="upperarm_l_null"), bChildInitial=False, bParentInitial=False).RelativeTransform.Rotation
    ).Twist) {template="Inverse::Execute(in Value,out Result)"}
   &RigUnit_ModifyTransforms(
    ItemToModify="((Item=(Type=Bone,Name=\"upperarm_twist_01_l\"),Transform=(Rotation=(X=0,Y=0,Z=0,W=1),Translation=(X=0,Y=0,Z=0),Scale3D=(X=1,Y=1,Z=1))),(Item=(Type=Bone,Name=\"upperarm_twist_02_l\"),Transform=(Rotation=(X=0,Y=0,Z=0,W=1),Translation=(X=0,Y=0,Z=0),Scale3D=(X=1, ...(truncated)"
    Weight=1
    WeightMinimum=0
    ItemToModify.1.Transform.Rotation=/Script/RigVM.RigVMFunction_MathQuaternionSlerp(B=(X=0,Y=0,Z=0,W=1), T=0.5, A=@n3.Result) {is=struct RigVMUnitNode, template="Interpolate::Execute(in A,in B,in T,out Result)"}.Result
    ItemToModify.0.Transform.Rotation=@n3.Result
    ExecutePin=@RigVMFunction_Sequence.C
   )
   n4 = &RigVMFunction_MathQuaternionInverse(Value=&RigVMFunction_MathQuaternionSwingTwist(
     TwistAxis=(X=1,Y=0,Z=0)
     Input=&RigUnit_GetRelativeTransformForItem(Child=(Type=Bone,Name="upperarm_r"), Parent=(Type=Null,Name="upperarm_r_null"), bChildInitial=False, bParentInitial=False).RelativeTransform.Rotation
    ).Twist) {template="Inverse::Execute(in Value,out Result)"}
   Interpolate = &RigVMFunction_MathQuaternionSlerp(B=(X=0,Y=0,Z=0,W=1), T=0.5, A=@n4.Result) {template="Interpolate::Execute(in A,in B,in T,out Result)"}
   > &RigUnit_ModifyTransforms(
    ItemToModify="((Item=(Type=Bone,Name=\"upperarm_twist_01_r\"),Transform=(Rotation=(X=0,Y=0,Z=0,W=1),Translation=(X=0,Y=0,Z=0),Scale3D=(X=1,Y=1,Z=1))),(Item=(Type=Bone,Name=\"upperarm_twist_02_r\"),Transform=(Rotation=(X=0,Y=0,Z=0,W=1),Translation=(X=0,Y=0,Z=0),Scale3D=(X=1, ...(truncated)"
    Weight=1
    WeightMinimum=0
    ItemToModify.1.Transform.Rotation=@Interpolate.Result
    ItemToModify.0.Transform.Rotation=@n4.Result
   )
  ]}
 New Function_ContainedGraph {nodes [
   RigVMFunctionEntryNode()
   > DeltaRotation = RigVMFunctionReferenceNode(
    Rotation Order=ZYX
    child=hand_l
    debugRotation=false
    max=100
    min=70
    onlyTwist=true
    parent=hand_l_null
    parentType=Null
    twistAxis=(X=1,Y=0,Z=0)
   )
   > &RigUnit_ModifyTransforms(
    ItemToModify="((Item=(Type=Bone,Name=\"lowerarm_twist_01_l\"),Transform=(Rotation=(X=0,Y=-0.0262,Z=0,W=0.9997),Translation=(X=-0.7,Y=-0.17,Z=1),Scale3D=(X=1,Y=1,Z=1))),(Item=(Type=Bone,Name=\"lowerarm_twist_02_l\"),Transform=(Rotation=(X=0.342,Y=0,Z=0,W=0.9397),Translation= ...(truncated)"
    WeightMinimum=0
    Weight=@DeltaRotation.outputWeight
   )
   > RigVMFunctionReturnNode()
  ]}
 New Function_ContainedGraph {nodes [
   RigVMFunctionEntryNode()
   > n1 = &RigUnit_SphericalPoseReader(
    ActiveRegionScaleFactors=(PositiveWidth=1.2,NegativeWidth=1,PositiveHeight=1,NegativeHeight=1)
    ActiveRegionSize=0.1
    Debug=(bDrawDebug=false,bDraw2D=False,bDrawLocalAxes=False,DebugScale=25,DebugSegments=20,DebugThickness=0.25)
    DriverAxis=(X=1,Y=0,Z=0)
    DriverItem=(Type=Bone,Name="upperarm_l")
    FalloffRegionScaleFactors=(PositiveWidth=2,NegativeWidth=1,PositiveHeight=1,NegativeHeight=1)
    FalloffSize=0.3
    FlipHeightScaling=False
    FlipWidthScaling=False
    OptionalParentItem=(Type=Bone,Name="None")
    RotationOffset=(X=80,Y=-90,Z=0)
   )
   > &RigUnit_ModifyTransforms(
    ItemToModify="((Item=(Type=Bone,Name=\"base_upperarm_l_sleeves_dw\"),Transform=(Rotation=(X=0,Y=0,Z=0,W=1),Translation=(X=5,Y=0,Z=2.1),Scale3D=(X=1,Y=1,Z=1))),(Item=(Type=Bone,Name=\"base_upperarm_l_sleeves_fw\"),Transform=(Rotation=(X=0,Y=0,Z=0,W=1),Translation=(X=2.2,Y=-0 ...(truncated)"
    WeightMinimum=0
    Weight=@n1.OutputParam
   )
   > n2 = &RigUnit_SphericalPoseReader(
    ActiveRegionScaleFactors=(PositiveWidth=1.2,NegativeWidth=1,PositiveHeight=1,NegativeHeight=1)
    ActiveRegionSize=0.1
    Debug=(bDrawDebug=false,bDraw2D=False,bDrawLocalAxes=False,DebugScale=25,DebugSegments=20,DebugThickness=0.25)
    DriverAxis=(X=-1,Y=0,Z=0)
    DriverItem=(Type=Bone,Name="upperarm_r")
    FalloffRegionScaleFactors=(PositiveWidth=2,NegativeWidth=1,PositiveHeight=1,NegativeHeight=1)
    FalloffSize=0.3
    FlipHeightScaling=False
    FlipWidthScaling=False
    OptionalParentItem=(Type=Bone,Name="None")
    RotationOffset=(X=-80,Y=90,Z=0)
   )
   > &RigUnit_ModifyTransforms(
    ItemToModify="((Item=(Type=Bone,Name=\"base_upperarm_r_sleeves_dw\"),Transform=(Rotation=(X=0,Y=0,Z=0,W=-1),Translation=(X=7.1645,Y=0,Z=4.4204),Scale3D=(X=1,Y=1,Z=1))),(Item=(Type=Bone,Name=\"base_upperarm_r_sleeves_fw\"),Transform=(Rotation=(X=0,Y=0,Z=0,W=1),Translation=(X ...(truncated)"
    WeightMinimum=0
    Weight=@n2.OutputParam
   )
   > n3 = &RigUnit_SphericalPoseReader(
    ActiveRegionScaleFactors=(PositiveWidth=1,NegativeWidth=1,PositiveHeight=1,NegativeHeight=1)
    ActiveRegionSize=0.05
    Debug=(bDrawDebug=false,bDraw2D=false,bDrawLocalAxes=false,DebugScale=25,DebugSegments=20,DebugThickness=0.1)
    DriverAxis=(X=1,Y=0,Z=0)
    DriverItem=(Type=Bone,Name="upperarm_l")
    FalloffRegionScaleFactors=(PositiveWidth=1,NegativeWidth=1,PositiveHeight=1,NegativeHeight=1)
    FalloffSize=0.2
    FlipHeightScaling=False
    FlipWidthScaling=False
    OptionalParentItem=(Type=Bone,Name="None")
    RotationOffset=(X=0,Y=-135,Z=20)
   )
   > &RigUnit_ModifyTransforms(
    ItemToModify=((Item=(Type=Bone,Name="base_upperarm_l_sleeves_dw"),Transform=(Rotation=(X=0,Y=0,Z=0,W=1),Translation=(X=1.8,Y=0,Z=2.5),Scale3D=(X=1,Y=1,Z=1))))
    WeightMinimum=0
    Weight=@n3.OutputParam
   )
   > n4 = &RigUnit_SphericalPoseReader(
    ActiveRegionScaleFactors=(PositiveWidth=1,NegativeWidth=1,PositiveHeight=1,NegativeHeight=1)
    ActiveRegionSize=0.05
    Debug=(bDrawDebug=false,bDraw2D=false,bDrawLocalAxes=false,DebugScale=25,DebugSegments=20,DebugThickness=0.1)
    DriverAxis=(X=-1,Y=0,Z=0)
    DriverItem=(Type=Bone,Name="upperarm_r")
    FalloffRegionScaleFactors=(PositiveWidth=1,NegativeWidth=1,PositiveHeight=1,NegativeHeight=1)
    FalloffSize=0.2
    FlipHeightScaling=False
    FlipWidthScaling=False
    OptionalParentItem=(Type=Bone,Name="None")
    RotationOffset=(X=0,Y=45,Z=20)
   )
   > &RigUnit_ModifyTransforms(
    ItemToModify=((Item=(Type=Bone,Name="base_upperarm_r_sleeves_dw"),Transform=(Rotation=(X=0,Y=0,Z=0,W=1),Translation=(X=1.8,Y=0,Z=2.5),Scale3D=(X=1,Y=1,Z=1))))
    WeightMinimum=0
    Weight=@n4.OutputParam
   )
   > RigVMFunctionReturnNode()
  ]}
 New Function_ContainedGraph {nodes [
   n1 = RigVMFunctionEntryNode()
   RigVMFunction_MathBoolToInteger = &RigVMFunction_MathBoolToInteger(Value=@n1.onlyTwist)
   RigVMFunction_MathQuaternionSwingTwist = &RigVMFunction_MathQuaternionSwingTwist(
    TwistAxis=@n1.twistAxis
    Input=&RigUnit_GetRelativeTransformForItem(
     Child=(Type=Bone,Name="None")
     Parent=(Type=Bone,Name="None")
     bChildInitial=False
     bParentInitial=False
     Child.Name=@n1.child
     Parent.Name=@n1.parent
     Parent.Type=@n1.parentType
    ).RelativeTransform.Rotation
   )
   RigVMFunction_MathQuaternionToEuler = &RigVMFunction_MathQuaternionToEuler(
    Value=RigVMDispatchNode(
     Values=((),())
     Index=@RigVMFunction_MathBoolToInteger.Result
     Values.0=@RigVMFunction_MathQuaternionSwingTwist.Swing
     Values.1=@RigVMFunction_MathQuaternionSwingTwist.Twist
    ) {template="DISPATCH_RigVMDispatch_SelectInt32(in Index,in Values,out Result)"}.Result
    RotationOrder=@n1.Rotation Order
   )
   > DISPATCH_RigVMDispatch_Print = RigVMDispatchNode(ScreenColor=(R=1,G=1,B=1,A=1), ScreenDuration=1.5, Enabled=@n1.debugRotation, Value=@RigVMFunction_MathQuaternionToEuler.Result) {template="DISPATCH_RigVMDispatch_Print(in Prefix,in Value,in Enabled,in ScreenDuration,in ScreenColor)"}
   Clamp = &RigVMFunction_MathIntClamp(Value=@n1.primAxis)
   DISPATCH_RigVMDispatch_SelectInt32 = RigVMDispatchNode(
    Values=(0,0,0)
    Index=@Clamp.Result
    Values.0=@RigVMFunction_MathQuaternionToEuler.Result.X
    Values.1=@RigVMFunction_MathQuaternionToEuler.Result.Y
    Values.2=@RigVMFunction_MathQuaternionToEuler.Result.Z
   ) {template="DISPATCH_RigVMDispatch_SelectInt32(in Index,in Values,out Result)"}
   Remap = &RigVMFunction_MathDoubleRemap(
    TargetMaximum=1
    TargetMinimum=0
    bClamp=False
    Value=@DISPATCH_RigVMDispatch_SelectInt32.Result
    SourceMaximum=@n1.max
    SourceMinimum=@n1.min
   )
   > RigVMFunctionReturnNode(outputWeight=&RigVMFunction_MathDoubleClamp(Value=@Remap.Result).Result)
  ]}
 New Function_ContainedGraph {nodes [
   RigVMFunctionEntryNode()
   > RigVMFunction_Sequence_2 = RigVMAggregateNode()
   &RigUnit_ParentConstraint(
    AdvancedSettings=(InterpolationType=Average,RotationOrderForFilter=XZY)
    Child=(Type=Bone,Name="base_L_skirt_a_1")
    Parents=((Item=(Type=Bone,Name="thigh_twist_01_l"),Weight=1))
    Weight=1
    ExecutePin=@RigVMFunction_Sequence_2.A
   )
   > &RigUnit_ParentConstraint(
    AdvancedSettings=(InterpolationType=Average,RotationOrderForFilter=XZY)
    Child=(Type=Bone,Name="base_L_skirt_b_1")
    Parents=((Item=(Type=Bone,Name="thigh_twist_01_l"),Weight=1))
    Weight=1
   )
   > &RigUnit_ParentConstraint(
    AdvancedSettings=(InterpolationType=Average,RotationOrderForFilter=XZY)
    Child=(Type=Bone,Name="base_L_skirt_c_1")
    Parents=((Item=(Type=Bone,Name="thigh_twist_01_l"),Weight=1))
    Weight=1
   )
   > &RigUnit_ParentConstraint(
    AdvancedSettings=(InterpolationType=Average,RotationOrderForFilter=XZY)
    Child=(Type=Bone,Name="base_R_skirt_a_1")
    Parents=((Item=(Type=Bone,Name="thigh_twist_01_r"),Weight=1))
    Weight=1
   )
   > &RigUnit_ParentConstraint(
    AdvancedSettings=(InterpolationType=Average,RotationOrderForFilter=XZY)
    Child=(Type=Bone,Name="base_R_skirt_b_1")
    Parents=((Item=(Type=Bone,Name="thigh_twist_01_r"),Weight=1))
    Weight=1
   )
   > &RigUnit_ParentConstraint(
    AdvancedSettings=(InterpolationType=Average,RotationOrderForFilter=XZY)
    Child=(Type=Bone,Name="base_R_skirt_c_1")
    Parents=((Item=(Type=Bone,Name="thigh_twist_01_r"),Weight=1))
    Weight=1
   )
   > &RigUnit_ParentConstraint(
    AdvancedSettings=(InterpolationType=Average,RotationOrderForFilter=XZY)
    Child=(Type=Bone,Name="base_C_skirt_a_1")
    Parents=((Item=(Type=Bone,Name="thigh_twist_01_l"),Weight=0.5),(Item=(Type=Bone,Name="thigh_twist_01_r"),Weight=0.5))
    Weight=1
   )
   > &RigUnit_ParentConstraint(
    AdvancedSettings=(InterpolationType=Average,RotationOrderForFilter=XZY)
    Child=(Type=Bone,Name="base_C_skirt_b_1")
    Parents=((Item=(Type=Bone,Name="thigh_twist_01_l"),Weight=0.5),(Item=(Type=Bone,Name="thigh_twist_01_r"),Weight=0.5))
    Weight=1
   )
   n1 = RigVMDispatchNode(BoneName=root, Default=0, Name=deformRig_l_thigh_up_alpha) {template="DISPATCH_RigDispatch_GetAnimAttribute(in Name,in BoneName,in Default,out Value,out Found)"}
   &RigUnit_ModifyTransforms(
    ItemToModify="((Item=(Type=Bone,Name=\"base_L_skirt_a_1\"),Transform=(Rotation=(X=-0.0012,Y=-0.0697,Z=-0.0174,W=0.9974),Translation=(X=0,Y=0,Z=0),Scale3D=(X=1,Y=1,Z=1))),(Item=(Type=Bone,Name=\"base_L_skirt_b_1\"),Transform=(Rotation=(X=-0.0027,Y=-0.0523,Z=-0.0523,W=0.9973) ...(truncated)"
    WeightMinimum=0
    Weight=@n1.Value
    ExecutePin=@RigVMFunction_Sequence_2.B
   )
   n2 = RigVMDispatchNode(BoneName=root, Default=0, Name=deformRig_r_thigh_up_alpha) {template="DISPATCH_RigDispatch_GetAnimAttribute(in Name,in BoneName,in Default,out Value,out Found)"}
   > &RigUnit_ModifyTransforms(
    ItemToModify="((Item=(Type=Bone,Name=\"base_R_skirt_a_1\"),Transform=(Rotation=(X=0.0012,Y=-0.0697,Z=0.0174,W=0.9974),Translation=(X=0,Y=0,Z=0),Scale3D=(X=1,Y=1,Z=1))),(Item=(Type=Bone,Name=\"base_R_skirt_b_1\"),Transform=(Rotation=(X=-0.0032,Y=-0.061,Z=-0.0522,W=0.9968),Tr ...(truncated)"
    WeightMinimum=0
    Weight=@n2.Value
   )
   Multiply = &RigVMFunction_MathDoubleMul(A=@n2.Value, B=@n1.Value)
   n3 = &RigVMFunction_AnimEvalRichCurve(
    Curve="(EditorCurveData=(Keys=((InterpMode=RCIM_Cubic,TangentMode=RCTM_User,ArriveTangent=1.6904,LeaveTangent=1.6904),(InterpMode=RCIM_Cubic,TangentMode=RCTM_User,Time=0.4643,Value=0.6384,ArriveTangent=1.0934,LeaveTangent=1.0934),(InterpMode=RCIM_Cubic,TangentMod ...(truncated)"
    SourceMinimum=0
    TargetMinimum=0
    Value=@Multiply.Result
   )
   > n4 = &RigUnit_ModifyTransforms(
    ItemToModify="((Item=(Type=Bone,Name=\"base_C_skirt_a_1\"),Transform=(Rotation=(X=0,Y=0.6428,Z=0,W=0.766),Translation=(X=0,Y=0,Z=0),Scale3D=(X=1,Y=1,Z=1))),(Item=(Type=Bone,Name=\"base_C_skirt_b_1\"),Transform=(Rotation=(X=0,Y=0.766,Z=0,W=0.6428),Translation=(X=0,Y=0,Z=0),S ...(truncated)"
    WeightMinimum=0
    Weight=@n3.Result
   )
   n5 = &RigVMFunction_ControlFlowBranch(ExecuteContext=@n4.ExecutePin, Condition=RigVMVariableNode() {variable=pub isCrouching}.Value)
   n6 = &RigUnit_SphericalPoseReader(
    ActiveRegionScaleFactors=(PositiveWidth=1,NegativeWidth=1,PositiveHeight=1,NegativeHeight=1)
    ActiveRegionSize=0.1
    Debug=(bDrawDebug=false,bDraw2D=False,bDrawLocalAxes=False,DebugScale=25,DebugSegments=20,DebugThickness=0.25)
    DriverAxis=(X=1,Y=0,Z=0)
    DriverItem=(Type=Bone,Name="thigh_r")
    FalloffRegionScaleFactors=(PositiveWidth=0.6,NegativeWidth=1,PositiveHeight=0.7,NegativeHeight=1)
    FalloffSize=0.3
    FlipHeightScaling=False
    FlipWidthScaling=False
    OptionalParentItem=(Type=Bone,Name="None")
    RotationOffset=(X=-65,Y=-170,Z=0)
    ExecutePin=@n5.True
   )
   n7 = /Script/RigVM.RigVMFunction_AnimRichCurve(Curve="(EditorCurveData=(Keys=((InterpMode=RCIM_Linear,TangentMode=RCTM_Auto,TangentWeightMode=RCTWM_WeightedNone,Time=0,Value=0,ArriveTangent=0,ArriveTangentWeight=0,LeaveTangent=0,LeaveTangentWeight=0),(InterpMode=RCIM_Cubic,TangentMode=RCTM_Auto,TangentWeightM ...(truncated)") {is=struct RigVMUnitNode}
   > RigVMVariableNode(Value=&RigVMFunction_AnimEvalRichCurve(SourceMinimum=0, TargetMinimum=0, Curve=@n7.Curve, Value=@n6.OutputParam).Result) {variable=r legSPR CrouchingStrafe float}
   > n8 = &RigUnit_SphericalPoseReader(
    ActiveRegionScaleFactors=(PositiveWidth=1,NegativeWidth=1,PositiveHeight=1,NegativeHeight=1)
    ActiveRegionSize=0.1
    Debug=(bDrawDebug=false,bDraw2D=False,bDrawLocalAxes=False,DebugScale=25,DebugSegments=20,DebugThickness=0.25)
    DriverAxis=(X=1,Y=0,Z=0)
    DriverItem=(Type=Bone,Name="thigh_l")
    FalloffRegionScaleFactors=(PositiveWidth=0.6,NegativeWidth=1,PositiveHeight=0.7,NegativeHeight=1)
    FalloffSize=0.3
    FlipHeightScaling=False
    FlipWidthScaling=False
    OptionalParentItem=(Type=Bone,Name="None")
    RotationOffset=(X=-65,Y=-170,Z=0)
   )
   > RigVMVariableNode(Value=&RigVMFunction_AnimEvalRichCurve(SourceMinimum=0, TargetMinimum=0, Value=@n8.OutputParam, Curve=@n7.Curve).Result) {variable=l legSPR CrouchingStrafe float}
   n9 = &RigVMFunction_MathDoubleMul(B=RigVMVariableNode() {variable=pub isCrouchStrafeLeft}.Value, A=RigVMVariableNode() {variable=r legSPR CrouchingStrafe float}.Value)
   > &RigUnit_ModifyTransforms(
    ItemToModify="((Item=(Type=Bone,Name=\"base_C_skirt_b_1\"),Transform=(Rotation=(X=0.2077,Y=-0.1458,Z=-0.0155,W=0.9671),Translation=(X=0,Y=0,Z=0),Scale3D=(X=1,Y=1,Z=1))),(Item=(Type=Bone,Name=\"base_C_skirt_b_2\"),Transform=(Rotation=(X=0,Y=-0.1392,Z=0,W=0.9903),Translation= ...(truncated)"
    WeightMinimum=0
    Weight=@n9.Result
   )
   > &RigUnit_ModifyTransforms(
    ItemToModify="((Item=(Type=Bone,Name=\"base_R_skirt_c_1\"),Transform=(Rotation=(X=0.1068,Y=-0.4281,Z=-0.4223,W=0.7918),Translation=(X=-3.5,Y=2.65,Z=-0.5),Scale3D=(X=1,Y=1,Z=1))),(Item=(Type=Bone,Name=\"base_R_skirt_b_1\"),Transform=(Rotation=(X=0.1914,Y=-0.2355,Z=-0.0846,W= ...(truncated)"
    WeightMinimum=0
    Weight=@n9.Result
   )
   > &RigUnit_ModifyTransforms(
    ItemToModify="((Item=(Type=Bone,Name=\"base_L_skirt_c_1\"),Transform=(Rotation=(X=0,Y=-0.0218,Z=0,W=0.9998),Translation=(X=0,Y=0,Z=0),Scale3D=(X=1,Y=1,Z=1))),(Item=(Type=Bone,Name=\"base_L_skirt_b_1\"),Transform=(Rotation=(X=0,Y=-0.0175,Z=0,W=0.9998),Translation=(X=0,Y=0,Z= ...(truncated)"
    WeightMinimum=0
    Weight=@n9.Result
   )
   n10 = &RigVMFunction_MathDoubleMul(B=RigVMVariableNode() {variable=pub isCrouchStrafeRight}.Value, A=RigVMVariableNode() {variable=l legSPR CrouchingStrafe float}.Value)
   > &RigUnit_ModifyTransforms(
    ItemToModify="((Item=(Type=Bone,Name=\"base_C_skirt_b_1\"),Transform=(Rotation=(X=0,Y=0,Z=0,W=1),Translation=(X=0,Y=0,Z=0),Scale3D=(X=1,Y=1,Z=1))),(Item=(Type=Bone,Name=\"base_C_skirt_b_2\"),Transform=(Rotation=(X=0,Y=0,Z=0,W=1),Translation=(X=0,Y=0,Z=0),Scale3D=(X=1,Y=1,Z= ...(truncated)"
    WeightMinimum=0
    Weight=@n10.Result
   )
   > &RigUnit_ModifyTransforms(
    ItemToModify="((Item=(Type=Bone,Name=\"base_R_skirt_c_1\"),Transform=(Rotation=(X=0,Y=0,Z=0,W=1),Translation=(X=0,Y=0,Z=0),Scale3D=(X=1,Y=1,Z=1))),(Item=(Type=Bone,Name=\"base_R_skirt_b_1\"),Transform=(Rotation=(X=0,Y=0,Z=0,W=1),Translation=(X=0,Y=0,Z=0),Scale3D=(X=1,Y=1,Z= ...(truncated)"
    WeightMinimum=0
    Weight=@n10.Result
   )
   > &RigUnit_ModifyTransforms(
    ItemToModify="((Item=(Type=Bone,Name=\"base_L_skirt_c_1\"),Transform=(Rotation=(X=-0.0206,Y=-0.1008,Z=0.1257,W=0.9867),Translation=(X=0,Y=0,Z=0),Scale3D=(X=1,Y=1,Z=1))),(Item=(Type=Bone,Name=\"base_L_skirt_b_1\"),Transform=(Rotation=(X=-0.0008,Y=-0.0799,Z=-0.0098,W=0.9968), ...(truncated)"
    WeightMinimum=0
    Weight=@n10.Result
   )
   > &RigUnit_ModifyTransforms(
    ItemToModify="((Item=(Type=Bone,Name=\"base_C_skirt_b_1\"),Transform=(Rotation=(X=-0.0353,Y=-0.1739,Z=0.0454,W=0.9831),Translation=(X=0,Y=0,Z=0),Scale3D=(X=1,Y=1,Z=1))),(Item=(Type=Bone,Name=\"base_L_skirt_c_1\"),Transform=(Rotation=(X=-0.2428,Y=-0.0091,Z=0.1367,W=0.9603),T ...(truncated)"
    WeightMinimum=0
    Weight=&RigVMFunction_MathDoubleMul(B=RigVMVariableNode() {variable=pub isCrouchStrafeLeft}.Value, A=RigVMVariableNode() {variable=l legSPR CrouchingStrafe float}.Value).Result
   )
   DeltaRotation_2 = RigVMFunctionReferenceNode(
    Rotation Order=ZXY
    child=thigh_l
    debugRotation=false
    max=90
    parent=thigh_l_null
    parentType=Null
    primAxis=2
    twistAxis=(X=1,Y=0,Z=0)
    ExecuteContext=@RigVMFunction_Sequence_2.C
   )
   > RigVMVariableNode(Value=@DeltaRotation_2.outputWeight) {variable=l leg backwards SPR float}
   > DeltaRotation_2_1 = RigVMFunctionReferenceNode(Rotation Order=ZXY, child=thigh_r, debugRotation=false, max=90, parent=thigh_r_null, parentType=Null, primAxis=2, twistAxis=(X=1,Y=0,Z=0))
   > RigVMVariableNode(Value=@DeltaRotation_2_1.outputWeight) {variable=r leg backwards SPR float}
   n11 = RigVMVariableNode() {variable=pub isZipline}
   Interpolate = /Script/RigVM.RigVMFunction_MathDoubleLerp(B=0, T=@n11.Value, A=RigVMVariableNode() {variable=l leg backwards SPR float}.Value) {is=struct RigVMUnitNode, template="Interpolate::Execute(in A,in B,in T,out Result)"}
   > &RigUnit_ModifyTransforms(
    ItemToModify="((Item=(Type=Bone,Name=\"base_L_skirt_a_1\"),Transform=(Rotation=(X=0,Y=-0.1736,Z=0,W=0.9848),Translation=(X=0,Y=0,Z=0),Scale3D=(X=1,Y=1,Z=1))),(Item=(Type=Bone,Name=\"base_L_skirt_c_1\"),Transform=(Rotation=(X=0,Y=-0.0436,Z=0,W=0.999),Translation=(X=0,Y=0,Z=0 ...(truncated)"
    WeightMinimum=0
    Weight=@Interpolate.Result
   )
   n12 = &RigVMFunction_MathDoubleLerp(B=0, T=@n11.Value, A=RigVMVariableNode() {variable=r leg backwards SPR float}.Value) {template="Interpolate::Execute(in A,in B,in T,out Result)"}
   > &RigUnit_ModifyTransforms(
    ItemToModify="((Item=(Type=Bone,Name=\"base_R_skirt_a_1\"),Transform=(Rotation=(X=0,Y=-0.1736,Z=0,W=0.9848),Translation=(X=0,Y=0,Z=0),Scale3D=(X=1,Y=1,Z=1))),(Item=(Type=Bone,Name=\"base_R_skirt_c_1\"),Transform=(Rotation=(X=0,Y=-0.0436,Z=0,W=0.999),Translation=(X=0,Y=0,Z=0 ...(truncated)"
    WeightMinimum=0
    Weight=@n12.Result
   )
   > &RigUnit_ModifyTransforms(
    ItemToModify="((Item=(Type=Bone,Name=\"base_C_skirt_a_1\"),Transform=(Rotation=(X=0,Y=0.5373,Z=0,W=0.8434),Translation=(X=0,Y=0,Z=0),Scale3D=(X=1,Y=1,Z=1))),(Item=(Type=Bone,Name=\"base_C_skirt_b_1\"),Transform=(Rotation=(X=0,Y=0.4067,Z=0,W=0.9135),Translation=(X=0,Y=0,Z=0) ...(truncated)"
    WeightMinimum=0
    Weight=&RigVMFunction_MathDoubleMul(A=@Interpolate.Result, B=@n12.Result).Result
   )
   RigVMFunction_Sequence = RigVMAggregateNode(ExecuteContext=@RigVMFunction_Sequence_2.E)
   n13 = &RigVMFunction_Sequence(ExecuteContext=@RigVMFunction_Sequence.D)
   RigVMFunction_MathBoolOr = &RigVMFunction_MathBoolOr(A=RigVMVariableNode() {variable=pub isCrouching}.Value, B=RigVMVariableNode() {variable=pub isPlayingEasternBloc}.Value)
   RigVMFunction_ControlFlowBranch = &RigVMFunction_ControlFlowBranch(Condition=@RigVMFunction_MathBoolOr.Result, ExecuteContext=@n13.A)
   RigUnit_GetRelativeTransformForItem = &RigUnit_GetRelativeTransformForItem(Child=(Type=Bone,Name="pelvis"), Parent=(Type=Bone,Name="root"), bChildInitial=False, bParentInitial=False)
   RigVMFunction_AnimEvalRichCurve = &RigVMFunction_AnimEvalRichCurve(
    Curve="(EditorCurveData=(Keys=((InterpMode=RCIM_Linear,TangentMode=RCTM_Auto,TangentWeightMode=RCTWM_WeightedNone,Time=0,Value=0,ArriveTangent=0,ArriveTangentWeight=0,LeaveTangent=0,LeaveTangentWeight=0),(InterpMode=RCIM_Cubic,TangentMode=RCTM_Auto,TangentWeightM ...(truncated)"
    SourceMaximum=45
    SourceMinimum=90
    TargetMinimum=0
    Value=@RigUnit_GetRelativeTransformForItem.RelativeTransform.Translation.Z
   )
   Multiply_5 = RigVMAggregateNode(
    B=&RigVMFunction_MathDoubleRemap(
     SourceMaximum=1
     SourceMinimum=0
     TargetMaximum=0
     TargetMinimum=1
     bClamp=False
     Value=RigVMVariableNode() {variable=pub isBaseballSliding}.Value
    ).Result
    C=&RigVMFunction_MathDoubleRemap(
     SourceMaximum=1
     SourceMinimum=0
     TargetMaximum=0
     TargetMinimum=1
     bClamp=False
     Value=RigVMVariableNode() {variable=pub isKneeSliding}.Value
    ).Result
    A=@RigVMFunction_AnimEvalRichCurve.Result
   )
   &RigUnit_ModifyTransforms(
    ItemToModify="((Item=(Type=Bone,Name=\"base_R_skirt_c_1\"),Transform=(Rotation=(X=0,Y=-0.1915,Z=0,W=0.9815),Translation=(X=-19.8322,Y=-0.0012,Z=10.9091),Scale3D=(X=1,Y=1,Z=1))),(Item=(Type=Bone,Name=\"base_R_skirt_c_2\"),Transform=(Rotation=(X=0.012,Y=-0.2351,Z=0.0912,W=-0. ...(truncated)"
    WeightMinimum=0
    Weight=@Multiply_5.Result
    ExecutePin=@RigVMFunction_ControlFlowBranch.True
   )
   > &RigUnit_ModifyTransforms(
    ItemToModify="((Item=(Type=Bone,Name=\"base_L_skirt_b_1\"),Transform=(Rotation=(X=-0.0042,Y=0.078,Z=-0.0572,W=0.9953),Translation=(X=0,Y=0,Z=0),Scale3D=(X=1,Y=1,Z=1))),(Item=(Type=Bone,Name=\"base_L_skirt_b_2\"),Transform=(Rotation=(X=0.0075,Y=0.1337,Z=0.0482,W=0.9898),Tran ...(truncated)"
    WeightMinimum=0
    Weight=@Multiply_5.Result
   )
   &RigUnit_ModifyTransforms(
    ItemToModify="((Item=(Type=Bone,Name=\"base_L_skirt_c_1\"),Transform=(Rotation=(X=0,Y=0,Z=0,W=1),Translation=(X=6.9632,Y=0,Z=-7.3486),Scale3D=(X=1,Y=1,Z=1))),(Item=(Type=Bone,Name=\"base_R_skirt_c_1\"),Transform=(Rotation=(X=0,Y=0,Z=0,W=1),Translation=(X=-2.8353,Y=0,Z=-0.86 ...(truncated)"
    WeightMinimum=0
    Weight=RigVMVariableNode() {variable=pub isKneeSliding}.Value
    ExecutePin=@n13.B
   )
   > &RigUnit_ModifyTransforms(
    ItemToModify="((Item=(Type=Bone,Name=\"base_L_skirt_c_1\"),Transform=(Rotation=(X=0,Y=-0.0654,Z=0,W=0.9979),Translation=(X=-1.75,Y=0,Z=7.5),Scale3D=(X=1,Y=1,Z=1))),(Item=(Type=Bone,Name=\"base_C_skirt_b_1\"),Transform=(Rotation=(X=0,Y=-0.4226,Z=0,W=0.9063),Translation=(X=-4 ...(truncated)"
    WeightMinimum=0
    Weight=RigVMVariableNode() {variable=pub isBaseballSliding}.Value
   )
   > DeltaRotation_3 = RigVMFunctionReferenceNode(
    Rotation Order=ZXY
    child=calf_l
    debugRotation=false
    max=128
    min=80
    parent=calf_l_null
    parentType=Null
    primAxis=2
    twistAxis=(X=1,Y=0,Z=0)
   )
   n14 = &RigVMFunction_AnimEvalRichCurve(
    Curve="(EditorCurveData=(Keys=((InterpMode=RCIM_Cubic),(InterpMode=RCIM_Cubic,TangentMode=RCTM_User,Time=0.6264,Value=0.3068,ArriveTangent=1.2597,LeaveTangent=1.2597),(InterpMode=RCIM_Cubic,TangentMode=RCTM_User,Time=1,Value=1,ArriveTangent=2.7015,LeaveTangent=2. ...(truncated)"
    SourceMaximum=38
    SourceMinimum=55
    TargetMinimum=0
    Value=&RigUnit_GetRelativeTransformForItem(Child=(Type=Bone,Name="pelvis"), Parent=(Type=Bone,Name="root"), bChildInitial=False, bParentInitial=False).RelativeTransform.Translation.Z
   )
   Multiply_3 = RigVMAggregateNode(A=@DeltaRotation_3.outputWeight, C=@n14.Result, B=RigVMVariableNode() {variable=pub isGrindRail}.Value)
   > &RigUnit_ModifyTransforms(
    ItemToModify="((Item=(Type=Bone,Name=\"base_L_skirt_c_2\"),Transform=(Rotation=(X=-0.0431,Y=-0.5686,Z=-0.0621,W=-0.8191),Translation=(X=4.6502,Y=3.5608,Z=6.8937),Scale3D=(X=1,Y=1,Z=1))),(Item=(Type=Bone,Name=\"base_C_skirt_b_2\"),Transform=(Rotation=(X=0,Y=-0.2067,Z=0,W=0.9 ...(truncated)"
    WeightMinimum=0
    Weight=@Multiply_3.Result
   )
   n15 = RigVMVariableNode() {variable=l thighLocal rz float}
   n16 = RigVMVariableNode() {variable=r thighLocal rz float}
   &RigUnit_ModifyTransforms(
    ItemToModify="((Item=(Type=Bone,Name=\"base_C_skirt_a_1\"),Transform=(Rotation=(X=0,Y=-0.1305,Z=0,W=0.9914),Translation=(X=0,Y=0,Z=0),Scale3D=(X=1,Y=1,Z=1))),(Item=(Type=Bone,Name=\"base_C_skirt_b_1\"),Transform=(Rotation=(X=0,Y=-0.1305,Z=0,W=0.9914),Translation=(X=0,Y=0,Z= ...(truncated)"
    WeightMinimum=0
    Weight=&RigVMFunction_MathDoubleMul(
     A=&RigVMFunction_MathDoubleRemap(SourceMaximum=45, SourceMinimum=0, TargetMaximum=1, TargetMinimum=0, Value=@n15.Value).Result
     B=&RigVMFunction_MathDoubleRemap(SourceMaximum=-45, SourceMinimum=0, TargetMaximum=1, TargetMinimum=0, Value=@n16.Value).Result
    ).Result
    ExecutePin=@RigVMFunction_Sequence.B
   )
   Remap = &RigVMFunction_MathDoubleRemap(SourceMaximum=-45, SourceMinimum=0, TargetMaximum=1, TargetMinimum=0, Value=@n15.Value)
   > &RigUnit_ModifyTransforms(
    ItemToModify="((Item=(Type=Bone,Name=\"base_C_skirt_a_1\"),Transform=(Rotation=(X=0,Y=-0.1305,Z=0,W=0.9914),Translation=(X=0,Y=0,Z=0),Scale3D=(X=1,Y=1,Z=1))),(Item=(Type=Bone,Name=\"base_C_skirt_b_1\"),Transform=(Rotation=(X=0,Y=-0.1305,Z=0,W=0.9914),Translation=(X=0,Y=0,Z= ...(truncated)"
    WeightMinimum=0
    Weight=&RigVMFunction_MathDoubleMul(
     A=@Remap.Result
     B=&RigVMFunction_MathDoubleRemap(SourceMaximum=45, SourceMinimum=0, TargetMaximum=1, TargetMinimum=0, Value=@n16.Value).Result
    ).Result
   )
   RigVMFunction_MathQuaternionSwingTwist = &RigVMFunction_MathQuaternionSwingTwist(
    TwistAxis=(X=1,Y=0,Z=0)
    Input=&RigUnit_GetRelativeTransformForItem(Child=(Type=Bone,Name="thigh_l"), Parent=(Type=Null,Name="thigh_l_null"), bChildInitial=False, bParentInitial=False).RelativeTransform.Rotation
   )
   RigVMFunction_MathQuaternionToEuler = &RigVMFunction_MathQuaternionToEuler(RotationOrder=ZYX, Value=@RigVMFunction_MathQuaternionSwingTwist.Swing)
   RigVMVariableNode(Value=@RigVMFunction_MathQuaternionToEuler.Result.Y, ExecuteContext=@RigVMFunction_Sequence.A) {variable=l thighLocal ry float}
   > RigVMVariableNode(Value=@RigVMFunction_MathQuaternionToEuler.Result.Z) {variable=l thighLocal rz float}
   n17 = &RigVMFunction_MathQuaternionToEuler(
    RotationOrder=ZYX
    Value=&RigVMFunction_MathQuaternionSwingTwist(
     TwistAxis=(X=1,Y=0,Z=0)
     Input=&RigUnit_GetRelativeTransformForItem(Child=(Type=Bone,Name="thigh_r"), Parent=(Type=Null,Name="thigh_r_null"), bChildInitial=False, bParentInitial=False).RelativeTransform.Rotation
    ).Swing
   )
   > RigVMVariableNode(Value=@n17.Result.Y) {variable=r thighLocal ry float}
   > RigVMVariableNode(Value=@n17.Result.Z) {variable=r thighLocal rz float}
   n18 = RigVMVariableNode() {variable=l thighLocal ry float}
   n19 = RigVMVariableNode() {variable=l thighLocal rz float}
   &RigUnit_ModifyTransforms(
    ItemToModify="((Item=(Type=Bone,Name=\"base_C_skirt_a_1\"),Transform=(Rotation=(X=0,Y=0,Z=0.1045,W=0.9945),Translation=(X=0,Y=0,Z=0),Scale3D=(X=1,Y=1,Z=1))),(Item=(Type=Bone,Name=\"base_C_skirt_b_1\"),Transform=(Rotation=(X=0,Y=0,Z=0.1305,W=0.9914),Translation=(X=0,Y=0,Z=0) ...(truncated)"
    WeightMinimum=0
    Weight=&RigVMFunction_MathDoubleMul(
     A=&RigVMFunction_MathDoubleRemap(SourceMaximum=24, SourceMinimum=0, TargetMaximum=1, TargetMinimum=0, Value=@n18.Value).Result
     B=&RigVMFunction_MathDoubleRemap(SourceMaximum=-70, SourceMinimum=0, TargetMaximum=1, TargetMinimum=0, Value=@n19.Value).Result
    ).Result
    ExecutePin=@RigVMFunction_Sequence.C
   )
   n20 = RigVMVariableNode() {variable=r thighLocal ry float}
   n21 = RigVMVariableNode() {variable=r thighLocal rz float}
   > &RigUnit_ModifyTransforms(
    ItemToModify="((Item=(Type=Bone,Name=\"base_C_skirt_a_1\"),Transform=(Rotation=(X=0,Y=0,Z=-0.1045,W=0.9945),Translation=(X=0,Y=0,Z=0),Scale3D=(X=1,Y=1,Z=1))),(Item=(Type=Bone,Name=\"base_C_skirt_b_1\"),Transform=(Rotation=(X=0,Y=0,Z=-0.1305,W=0.9914),Translation=(X=0,Y=0,Z= ...(truncated)"
    WeightMinimum=0
    Weight=&RigVMFunction_MathDoubleMul(
     A=&RigVMFunction_MathDoubleRemap(SourceMaximum=24, SourceMinimum=0, TargetMaximum=1, TargetMinimum=0, Value=@n20.Value).Result
     B=&RigVMFunction_MathDoubleRemap(SourceMaximum=-70, SourceMinimum=0, TargetMaximum=1, TargetMinimum=0, Value=@n21.Value).Result
    ).Result
   )
   RigVMFunctionReturnNode(ExecuteContext=@RigVMFunction_Sequence_2.F)
   n22 = &RigVMFunction_ControlFlowBranch(Condition=RigVMVariableNode() {variable="pub isDBNO "}.Value, ExecuteContext=@RigVMFunction_Sequence_2.D)
   DeltaRotation_0 = RigVMFunctionReferenceNode(
    Rotation Order=ZYX
    child=calf_l
    max=120
    parent=calf_l_null
    parentType=Null
    primAxis=2
    twistAxis=(X=1,Y=0,Z=0)
    ExecuteContext=@n22.False
   )
   RigVMFunction_AnimRichCurve = &RigVMFunction_AnimRichCurve(Curve="(EditorCurveData=(Keys=((InterpMode=RCIM_Cubic,TangentMode=RCTM_User,ArriveTangent=1.5462,LeaveTangent=0.3774),(InterpMode=RCIM_Cubic,Time=0.5,Value=0.379,ArriveTangent=1,LeaveTangent=1),(InterpMode=RCIM_Cubic,TangentMode=RCTM_User,Time=1,Value=1,ArriveTan ...(truncated)")
   n23 = &RigVMFunction_AnimEvalRichCurve(SourceMinimum=0, TargetMinimum=0, Value=@DeltaRotation_0.outputWeight, Curve=@RigVMFunction_AnimRichCurve.Curve)
   > &RigUnit_ModifyTransforms(
    ItemToModify="((Item=(Type=Bone,Name=\"base_L_skirt_c_2\"),Transform=(Rotation=(X=0,Y=-0.9063,Z=0,W=0.4226),Translation=(X=-4.2,Y=0,Z=0),Scale3D=(X=0.5,Y=1,Z=1))),(Item=(Type=Bone,Name=\"base_L_skirt_b_2\"),Transform=(Rotation=(X=0,Y=-0.1736,Z=0,W=0.9848),Translation=(X=0,Y ...(truncated)"
    WeightMinimum=0
    Weight=@n23.Result
   )
   > DeltaRotation_0_1 = RigVMFunctionReferenceNode(Rotation Order=ZYX, child=calf_r, max=120, parent=calf_r_null, parentType=Null, primAxis=2, twistAxis=(X=1,Y=0,Z=0))
   n24 = &RigVMFunction_AnimEvalRichCurve(SourceMinimum=0, TargetMinimum=0, Value=@DeltaRotation_0_1.outputWeight, Curve=@RigVMFunction_AnimRichCurve.Curve)
   > &RigUnit_ModifyTransforms(
    ItemToModify="((Item=(Type=Bone,Name=\"base_R_skirt_c_2\"),Transform=(Rotation=(X=0,Y=-0.9063,Z=0,W=0.4226),Translation=(X=-4.2,Y=0,Z=0),Scale3D=(X=0.5,Y=1,Z=1))),(Item=(Type=Bone,Name=\"base_R_skirt_b_2\"),Transform=(Rotation=(X=0,Y=-0.1736,Z=0,W=0.9848),Translation=(X=0,Y ...(truncated)"
    WeightMinimum=0
    Weight=@n24.Result
   )
   > &RigUnit_ModifyTransforms(
    ItemToModify="((Item=(Type=Bone,Name=\"base_C_skirt_b_2\"),Transform=(Rotation=(X=0,Y=-0.2588,Z=0,W=0.9659),Translation=(X=0,Y=0,Z=1),Scale3D=(X=0.75,Y=1,Z=1))),(Item=(Type=Bone,Name=\"base_L_skirt_c_2\"),Transform=(Rotation=(X=0,Y=0.2588,Z=0,W=0.9659),Translation=(X=0,Y=0, ...(truncated)"
    WeightMinimum=0
    Weight=&RigVMFunction_MathDoubleMul(A=@n24.Result, B=@n23.Result).Result
   )
   > DeltaRotation_0_2 = RigVMFunctionReferenceNode(Rotation Order=ZYX, child=calf_l, max=140, min=120, parent=calf_l_null, parentType=Null, primAxis=2, twistAxis=(X=1,Y=0,Z=0))
   > &RigUnit_ModifyTransforms(
    ItemToModify=((Item=(Type=Bone,Name="base_L_skirt_c_2"),Transform=(Rotation=(X=-0.2536,Y=-0.1926,Z=-0.0516,W=0.9465),Translation=(X=10.3,Y=6.6,Z=18),Scale3D=(X=1,Y=1,Z=1))))
    WeightMinimum=0
    Weight=@DeltaRotation_0_2.outputWeight
   )
   > DeltaRotation_0_1_1 = RigVMFunctionReferenceNode(Rotation Order=ZYX, child=calf_r, max=140, min=120, parent=calf_r_null, parentType=Null, primAxis=2, twistAxis=(X=1,Y=0,Z=0))
   > &RigUnit_ModifyTransforms(
    ItemToModify=((Item=(Type=Bone,Name="base_R_skirt_c_2"),Transform=(Rotation=(X=0.2536,Y=-0.1926,Z=0.0516,W=0.9465),Translation=(X=-10.3,Y=-6.6,Z=18),Scale3D=(X=1,Y=1,Z=1))))
    WeightMinimum=0
    Weight=@DeltaRotation_0_1_1.outputWeight
   )
   > &RigUnit_ModifyTransforms(
    ItemToModify=((Item=(Type=Bone,Name="base_C_skirt_b_2"),Transform=(Rotation=(X=0,Y=0,Z=0,W=-1),Translation=(X=0,Y=0,Z=4.636),Scale3D=(X=1,Y=1,Z=1))))
    WeightMinimum=0
    Weight=&RigVMFunction_MathDoubleMul(B=@DeltaRotation_0_2.outputWeight, A=@DeltaRotation_0_1_1.outputWeight).Result
   )
   DeltaRotation = RigVMFunctionReferenceNode(
    Rotation Order=ZXY
    child=calf_l
    debugRotation=false
    max=126
    min=72
    parent=calf_l_null
    parentType=Null
    primAxis=2
    twistAxis=(X=1,Y=0,Z=0)
    ExecuteContext=@n22.True
   )
   > DeltaRotation_1 = RigVMFunctionReferenceNode(Rotation Order=ZXY, child=calf_r, max=126, min=72, parent=calf_r_null, parentType=Null, primAxis=2, twistAxis=(X=1,Y=0,Z=0))
   > &RigUnit_ModifyTransforms(
    ItemToModify=((Item=(Type=Bone,Name="base_L_skirt_c_2"),Transform=(Rotation=(X=0,Y=-0.8192,Z=0,W=0.5736),Translation=(X=0,Y=0,Z=0),Scale3D=(X=1,Y=1,Z=1))))
    WeightMinimum=0
    Weight=@DeltaRotation.outputWeight
   )
   > &RigUnit_ModifyTransforms(
    ItemToModify=((Item=(Type=Bone,Name="base_R_skirt_c_2"),Transform=(Rotation=(X=0,Y=-0.9063,Z=0,W=0.4226),Translation=(X=0,Y=0,Z=0),Scale3D=(X=1,Y=1,Z=1))))
    WeightMinimum=0
    Weight=@DeltaRotation_1.outputWeight
   )
   > &RigUnit_ModifyTransforms(
    ItemToModify=((Item=(Type=Bone,Name="base_C_skirt_b_2"),Transform=(Rotation=(X=0,Y=-0.9063,Z=0,W=0.4226),Translation=(X=0,Y=0,Z=0),Scale3D=(X=1,Y=1,Z=1))))
    WeightMinimum=0
    Weight=&RigVMFunction_MathDoubleMul(A=@DeltaRotation_1.outputWeight, B=@DeltaRotation.outputWeight).Result
   )
   &RigUnit_ModifyTransforms(
    ExecutePin=()
    ItemToModify="((Item=(Type=Bone,Name=\"base_L_skirt_c_1\"),Transform=(Rotation=(X=-0.0489,Y=-0.495,Z=0.0608,W=0.8654),Translation=(X=-8.35,Y=0,Z=-14),Scale3D=(X=1,Y=1,Z=1))),(Item=(Type=Bone,Name=\"base_R_skirt_c_1\"),Transform=(Rotation=(X=0,Y=-0.5,Z=0,W=0.866),Translation ...(truncated)"
    Weight=1
    WeightMinimum=0
   )
   &RigUnit_ModifyTransforms(
    ExecutePin=()
    ItemToModify="((Item=(Type=Bone,Name=\"base_R_skirt_c_1\"),Transform=(Rotation=(X=0,Y=-0.1778,Z=0,W=0.9841),Translation=(X=-17.1784,Y=-0.0001,Z=4.422),Scale3D=(X=1,Y=1,Z=1))),(Item=(Type=Bone,Name=\"base_R_skirt_c_2\"),Transform=(Rotation=(X=-0.0089,Y=0.2681,Z=-0.0916,W=0.9 ...(truncated)"
    Weight=0
    WeightMinimum=0
   )
   &RigUnit_SphericalPoseReader(
    ActiveRegionScaleFactors=(PositiveWidth=0.5,NegativeWidth=0.5,PositiveHeight=0.5,NegativeHeight=0.5)
    ActiveRegionSize=0.3
    Debug=(bDrawDebug=false,DebugScale=25,DebugSegments=20,DebugThickness=0.25)
    DriverAxis=(X=1,Y=0,Z=0)
    DriverItem=(Type=Bone,Name="thigh_r")
    ExecutePin=()
    FalloffRegionScaleFactors=(PositiveWidth=1,NegativeWidth=1,PositiveHeight=1,NegativeHeight=1)
    FalloffSize=0.5
    FlipHeightScaling=false
    FlipWidthScaling=true
    OptionalParentItem=(Type=Bone)
    RotationOffset=(X=-90,Y=0,Z=180)
   )
   &RigUnit_SphericalPoseReader(
    ActiveRegionScaleFactors=(PositiveWidth=0.5,NegativeWidth=0.5,PositiveHeight=0.5,NegativeHeight=0.5)
    ActiveRegionSize=0.3
    Debug=(bDrawDebug=false,DebugScale=25,DebugSegments=20,DebugThickness=0.25)
    DriverAxis=(X=-1,Y=0,Z=0)
    DriverItem=(Type=Bone,Name="thigh_l")
    ExecutePin=()
    FalloffRegionScaleFactors=(PositiveWidth=1,NegativeWidth=1,PositiveHeight=1,NegativeHeight=1)
    FalloffSize=0.5
    OptionalParentItem=(Type=Bone)
    RotationOffset=(X=90,Y=0,Z=180)
   )
  ]}
 New Function_ContainedGraph {nodes [
   n1 = RigVMFunctionEntryNode()
   RigVMFunction_Sequence = &RigVMFunction_Sequence(ExecuteContext=@n1.ExecuteContext)
   RigVMFunction_MathBoolNot = &RigVMFunction_MathBoolNot(Value=RigVMVariableNode() {variable=pub isPlayingEmote}.Value)
   RigVMFunction_MathBoolAnd = &RigVMFunction_MathBoolAnd(A=@RigVMFunction_MathBoolNot.Result, B=RigVMVariableNode() {variable=pub isFrontEnd}.Value)
   RigVMFunction_ControlFlowBranch = &RigVMFunction_ControlFlowBranch(Condition=@RigVMFunction_MathBoolAnd.Result, ExecuteContext=@RigVMFunction_Sequence.A)
   DeltaRotation_1 = RigVMFunctionReferenceNode(
    Rotation Order=ZYX
    child=lowerarm_r
    max=75
    min=50
    parent=upperarm_r
    parentType=Bone
    primAxis=2
    twistAxis=(X=1,Y=0,Z=0)
    ExecuteContext=@RigVMFunction_ControlFlowBranch.True
   )
   Greater = /Script/RigVM.RigVMFunction_MathDoubleGreater(B=0.8, A=@DeltaRotation_1.outputWeight) {is=struct RigVMUnitNode, template="Greater::Execute(in A,in B,out Result)"}
   RigVMFunction_MathVectorDistance = &RigVMFunction_MathVectorDistance(
    A=&RigUnit_GetTransform(Item=(Type=Bone,Name="hand_l"), bInitial=False).Transform.Translation
    B=&RigUnit_GetTransform(Item=(Type=Bone,Name="base_L_skirt_b_1"), bInitial=False).Transform.Translation
   )
   Less = /Script/RigVM.RigVMFunction_MathDoubleLess(B=11.5, A=@RigVMFunction_MathVectorDistance.Result) {is=struct RigVMUnitNode, template="Less::Execute(in A,in B,out Result)"}
   RigVMFunction_MathBoolAnd_1 = RigVMAggregateNode(
    A=@Greater.Result
    C=&RigVMFunction_MathDoubleGreater(B=10.5, A=@RigVMFunction_MathVectorDistance.Result) {template="Greater::Execute(in A,in B,out Result)"}.Result
    B=@Less.Result
   )
   n2 = &RigVMFunction_ControlFlowBranch(ExecuteContext=@DeltaRotation_1.ExecuteContext, Condition=@RigVMFunction_MathBoolAnd_1.Result)
   DeltaRotation = RigVMFunctionReferenceNode(
    Rotation Order=ZYX
    child=lowerarm_l
    max=75
    min=50
    parent=upperarm_l
    parentType=Bone
    primAxis=2
    twistAxis=(X=1,Y=0,Z=0)
    ExecuteContext=@n2.True
   )
   > &RigUnit_ModifyTransforms(
    ItemToModify="((Item=(Type=Bone,Name=\"upperarm_l\"),Transform=(Rotation=(X=0,Y=0.0262,Z=0,W=0.9997),Translation=(X=0,Y=0,Z=0),Scale3D=(X=1,Y=1,Z=1))),(Item=(Type=Bone,Name=\"thumb_02_l\"),Transform=(Rotation=(X=0,Y=0,Z=-0.0872,W=0.9962),Translation=(X=0,Y=0,Z=0),Scale3D=(X ...(truncated)"
    WeightMinimum=0
    Weight=@DeltaRotation.outputWeight
   )
   RigVMFunctionReturnNode(ExecuteContext=@RigVMFunction_Sequence.B)
  ]}
 RigVMFunctionLibrary {nodes [
   DeltaRotation = RigVMCollapseNode(ExecuteContext=(), Rotation Order=ZYX, twistAxis=(X=1,Y=0,Z=0))
   FixHandsOnHip_frontEnd = RigVMCollapseNode(ExecuteContext=())
   SkirtFix = RigVMCollapseNode(ExecuteContext=())
   SleevesFix = RigVMCollapseNode(ExecuteContext=())
   TwistsFix = RigVMCollapseNode(ExecuteContext=())
   controllers = RigVMCollapseNode(ExecuteContext=(), debug=False)
  ]}
 RigVMFunction_MathBoolAnd_1_1_ContainedGraph {nodes [
   n1 = RigVMFunctionEntryNode()
   RigVMFunctionReturnNode(Result=&RigVMFunction_MathBoolAnd(B=@n1.C, A=&RigVMFunction_MathBoolAnd(A=@n1.A, B=@n1.B).Result).Result)
  ]}
 RigVMFunction_Sequence_1_ContainedGraph {nodes [
   n1 = RigVMFunctionEntryNode()
   RigVMFunction_Sequence = &RigVMFunction_Sequence(ExecuteContext=@n1.ExecuteContext)
   n2 = &RigVMFunction_Sequence(ExecuteContext=@RigVMFunction_Sequence.B)
   n3 = &RigVMFunction_Sequence(ExecuteContext=@n2.B)
   n4 = &RigVMFunction_Sequence(ExecuteContext=@n3.B)
   n5 = &RigVMFunction_Sequence(ExecuteContext=@n4.B)
   RigVMFunctionReturnNode(A=@RigVMFunction_Sequence.A, B=@n2.A, C=@n3.A, D=@n4.A, E=@n5.A, F=@n5.B)
  ]}
 RigVMFunction_Sequence_1_ContainedGraph {nodes [
   n1 = RigVMFunctionEntryNode()
   RigVMFunction_Sequence = &RigVMFunction_Sequence(ExecuteContext=@n1.ExecuteContext)
   n2 = &RigVMFunction_Sequence(ExecuteContext=@RigVMFunction_Sequence.B)
   n3 = &RigVMFunction_Sequence(ExecuteContext=@n2.B)
   RigVMFunctionReturnNode(A=@RigVMFunction_Sequence.A, B=@n2.A, C=@n3.A, D=@n3.B)
  ]}
 RigVMFunction_Sequence_2_ContainedGraph {nodes [
   n1 = RigVMFunctionEntryNode()
   RigVMFunction_Sequence = &RigVMFunction_Sequence(ExecuteContext=@n1.ExecuteContext)
   n2 = &RigVMFunction_Sequence(ExecuteContext=@RigVMFunction_Sequence.B)
   RigVMFunctionReturnNode(A=@RigVMFunction_Sequence.A, B=@n2.A, C=@n2.B)
  ]}
 RigVMModel {nodes [
   RigUnit_BeginExecution = &RigUnit_BeginExecution()
   TwistsFix = RigVMFunctionReferenceNode(ExecuteContext=@RigUnit_BeginExecution.ExecutePin)
   > SleevesFix = RigVMFunctionReferenceNode()
   > FN_DeformRig_FNC = RigVMFunctionReferenceNode(
    DeformRig_ Curve="(deformRig_l_upperarm_up_alpha_72_#1=False,deformRig_r_upperarm_up_alpha_73_#2=False,deformRig_l_upperarm_fr_alpha_74_#3=False,deformRig_r_upperarm_fr_alpha_75_CB51B8 ...(truncated)"
    DeformRig_Bool="(upperarm_1_#4=True,elbow_3_#5=True,wrist_7_#6=True,groin_8_#7=True,glute_17_#8=True,knee_9_40C408324CA261 ...(truncated)"
    DeformRig_Debug=(upperarm_debug_18_#4=false,wrist_debug_20_#6=False,lat_debug_26_#9=False,groinGlute_debug_27_#7=False)
    DeformRig_Value="(upperarm_up_18_#10=(Rotation=(X=0,Y=0,Z=0,W=1),Translation=(X=0,Y=0,Z=2),Scale3D=(X=1,Y=1,Z=1)),upperarm_fwd_fr_19_#11=(Rotation=(X=0,Y=0,Z=0,W=1),Translation=(X=0,Y=14,Z=0),Scale3D=(X=0.5,Y=1,Z=1) ...(truncated)"
   )
   > SkirtFix = RigVMFunctionReferenceNode()
   > FixHandsOnHip_frontEnd = RigVMFunctionReferenceNode()
   controllers = RigVMFunctionReferenceNode(ExecuteContext=(), debug=RigVMVariableNode() {variable=pub disablecontrolRig}.Value)
  ]}
]
hierarchy [
 root {loc=<0 0 0>, rot=<0 0 0 1>}
 pelvis {parent=root, loc=<0 -0.5301 98.0571>, rot=<0.0127 0.707 -0.0127 -0.707>}
 spine_01 {parent=pelvis, loc=<6.1208 -0.9328 0>, rot=<0 0 0.034 -0.9994>}
 spine_02 {parent=spine_01, loc=<6.5996 -0.2095 0>, rot=<0 0 0.0238 -0.9997>}
 spine_03 {parent=spine_02, loc=<6.4914 -0.1237 0>, rot=<0 0 0.0149 -0.9999>}
 spine_04 {parent=spine_03, loc=<6.9598 0.0794 0>, rot=<0 0 -0.0132 -0.9999>}
 spine_05 {parent=spine_04, loc=<6.9987 0 0>, rot=<0 0 -0.0015 -1>}
 attach_backpack {parent=spine_05, loc=<0 -8.35 0>, rot=<-0.0093 -0.707 0.0093 -0.707>}
 neck_01 {parent=spine_05, loc=<11.3138 0.1529 0>, rot=<0 0 -0.2003 -0.9797>}
 neck_02 {parent=neck_01, loc=<4.5627 0 0>, rot=<0 0 0.0504 -0.9987>}
 head {parent=neck_02, loc=<5.3914 0 0>, rot=<0 0 0.115 -0.9934>}
 base_C_hair_1 {parent=head, loc=<6.9517 -5.4626 0>, rot=<-0.0949 0.7007 0.7007 -0.0949>}
 dyn_C_hair_1 {parent=base_C_hair_1, loc=<0 0 0>, rot=<0 0 0 1>}
 dyn_C_hair_2 {parent=dyn_C_hair_1, loc=<10.6015 0 0>, rot=<0 -0.0172 0 -0.9999>}
 dyn_C_hair_3 {parent=dyn_C_hair_2, loc=<10.6008 0 0>, rot=<0 0.0789 0 -0.9969>}
 dyn_C_hair_4 {parent=dyn_C_hair_3, loc=<10.5829 0 0>, rot=<0 0.0898 0 -0.996>}
 dyn_C_hair_5 {parent=dyn_C_hair_4, loc=<10.5971 0 0>, rot=<0 0.0001 0 -1>}
 dyn_C_hair_6 {parent=dyn_C_hair_5, loc=<10.5971 0 0>, rot=<0 -0.0903 0 -0.9959>}
 dyn_C_hair_7 {parent=dyn_C_hair_6, loc=<10.5831 0 0>, rot=<0 -0.0859 0 -0.9963>}
 dyn_C_hair_7_end {parent=dyn_C_hair_7, loc=<10.5979 0 0>, rot=<0 -0.0983 0 -0.9952>}
 base_L_hair_1 {parent=head, loc=<6.9314 -3.0124 -7.9806>, rot=<0.0956 -0.7006 -0.7001 0.0996>}
 dyn_L_hair_1 {parent=base_L_hair_1, loc=<0 0 0>, rot=<0 0 0 1>}
 dyn_L_hair_2 {parent=dyn_L_hair_1, loc=<10.7733 0 0>, rot=<0.0014 -0.0775 0.0234 -0.9967>}
 dyn_L_hair_3 {parent=dyn_L_hair_2, loc=<10.7695 0 0>, rot=<-0.0067 0.1556 0.0018 -0.9878>}
 dyn_L_hair_4 {parent=dyn_L_hair_3, loc=<10.67 0 0>, rot=<-0.0019 0.0735 -0.0194 -0.9971>}
 dyn_L_hair_5 {parent=dyn_L_hair_4, loc=<10.7742 0 0>, rot=<-0.0002 0.0115 0.0111 -0.9999>}
 dyn_L_hair_6 {parent=dyn_L_hair_5, loc=<10.7637 0 0>, rot=<0.004 -0.1481 -0.0009 -0.989>}
 dyn_L_hair_7 {parent=dyn_L_hair_6, loc=<10.7571 0 0>, rot=<0.0021 -0.0478 0.0183 -0.9987>}
 dyn_L_hair_7_end {parent=dyn_L_hair_7, loc=<10.7622 0 0>, rot=<0 -0.0983 0 -0.9952>}
 base_R_hair_1 {parent=head, loc=<6.9319 -3.0124 7.9806>, rot=<-0.0996 0.7001 0.7006 -0.0956>}
 dyn_R_hair_1 {parent=base_R_hair_1, loc=<0 0 0>, rot=<0 0 0 1>}
 dyn_R_hair_2 {parent=dyn_R_hair_1, loc=<10.7735 0 0.0001>, rot=<-0.0014 -0.0775 -0.0234 -0.9967>}
 dyn_R_hair_3 {parent=dyn_R_hair_2, loc=<10.7696 0 0>, rot=<0.0067 0.1556 -0.0018 -0.9878>}
 dyn_R_hair_4 {parent=dyn_R_hair_3, loc=<10.6703 0 0.0001>, rot=<0.0019 0.0735 0.0194 -0.9971>}
 dyn_R_hair_5 {parent=dyn_R_hair_4, loc=<10.7743 0 0>, rot=<0.0002 0.0115 -0.0111 -0.9999>}
 dyn_R_hair_6 {parent=dyn_R_hair_5, loc=<10.7632 0 0>, rot=<-0.004 -0.1481 0.0009 -0.989>}
 dyn_R_hair_7 {parent=dyn_R_hair_6, loc=<10.7574 0 0>, rot=<-0.0021 -0.0478 -0.0183 -0.9987>}
 dyn_R_hair_7_end {parent=dyn_R_hair_7, loc=<10.7622 0 0.0001>, rot=<0 -0.0983 0 -0.9952>}
 attach_cape {parent=spine_05, loc=<9.3216 -8.4501 0>, rot=<0.0616 -0.7044 -0.0616 -0.7044>}
 clavicle_l {parent=spine_05, loc=<5.4831 -0.1053 -3.2946>, rot=<0.0465 -0.7289 -0.009 -0.6829>}
 upperarm_l {parent=clavicle_l, loc=<9.8658 -1.2136 -0.4105>, rot=<0.0164 -0.3245 -0.0269 -0.9454>}
 lowerarm_l {parent=upperarm_l, loc=<26.465 -0.8345 0.259>, rot=<0.0036 0.0161 -0.2298 -0.9731>}
 hand_l {parent=lowerarm_l, loc=<22.7355 -0.3203 -0.6105>, rot=<-0.6784 -0.0223 0.0735 -0.7307>}
 thumb_01_l {parent=hand_l, loc=<1.3118 -0.3191 -2.459>, rot=<0.6546 -0.4042 0.0664 -0.6354>}
 thumb_02_l {parent=thumb_01_l, loc=<5.9336 -0.0448 0.0401>, rot=<-0.0037 0.0213 0.298 -0.9543>}
 thumb_03_l {parent=thumb_02_l, loc=<2.2223 0.3682 -0.0907>, rot=<0.0118 -0.013 0.0066 -0.9998>}
 index_metacarpal_l {parent=hand_l, loc=<1.6535 0.0183 -2.96>, rot=<0.0158 -0.0124 -0.0173 -0.9996>}
 index_01_l {parent=index_metacarpal_l, loc=<8.3008 0.0357 -0.0524>, rot=<-0.0044 0.0485 0.0924 -0.9945>}
 index_02_l {parent=index_01_l, loc=<3.3587 -0.3004 0.0452>, rot=<-0.002 0.0027 0.3225 -0.9466>}
 index_03_l {parent=index_02_l, loc=<2.5578 -0.123 -0.0252>, rot=<-0.002 0.0067 0.2558 -0.9667>}
 middle_metacarpal_l {parent=hand_l, loc=<2.1856 -0.1725 -0.5984>, rot=<0.006 0.0168 -0.0329 -0.9993>}
 middle_01_l {parent=middle_metacarpal_l, loc=<7.3275 -0.2447 0.0129>, rot=<-0.0208 0.0042 0.0858 -0.9961>}
 middle_02_l {parent=middle_01_l, loc=<4.0123 0.1401 0.1175>, rot=<-0.0218 0.0258 0.3512 -0.9357>}
 middle_03_l {parent=middle_02_l, loc=<2.6077 0.0534 0.1373>, rot=<0.0082 0.0188 0.1708 -0.9851>}
 ring_metacarpal_l {parent=hand_l, loc=<2.2074 -0.3625 1.3638>, rot=<-0.0284 0.0722 -0.0458 -0.9959>}
 ring_01_l {parent=ring_metacarpal_l, loc=<7.0295 -0.6197 -0.1446>, rot=<-0.0489 -0.0252 0.1892 -0.9804>}
 ring_02_l {parent=ring_01_l, loc=<3.6174 0.4647 -0.1027>, rot=<0 0 0.2806 -0.9598>}
 ring_03_l {parent=ring_02_l, loc=<2.4152 0.114 0.0199>, rot=<0 0 0.1189 -0.9929>}
 pinky_metacarpal_l {parent=hand_l, loc=<1.7616 -0.9753 3.3814>, rot=<-0.0397 0.0814 -0.069 -0.9935>}
 pinky_01_l {parent=pinky_metacarpal_l, loc=<7.1709 -0.4831 -0.1094>, rot=<-0.0654 -0.0035 0.2242 -0.9723>}
 pinky_02_l {parent=pinky_01_l, loc=<2.8274 0.0893 -0.1043>, rot=<0.0037 0.0011 0.2854 -0.9584>}
 pinky_03_l {parent=pinky_02_l, loc=<2.2046 0.0307 -0.0078>, rot=<-0.0002 0.0098 0.1445 -0.9895>}
 deform_wrist_dn_base_l {parent=hand_l, loc=<0 -1.5326 0>, rot=<0 0 0 -1>}
 deform_wrist_dn_l {parent=deform_wrist_dn_base_l, loc=<0 0 0>, rot=<0 0 0 1>}
 deform_wrist_up_base_l {parent=hand_l, loc=<0 2.3009 0>, rot=<0 0 0 -1>}
 deform_wrist_up_l {parent=deform_wrist_up_base_l, loc=<0 0 0>, rot=<0 0 0 1>}
 hand_attach_l {parent=hand_l, loc=<8.5 -3 0>, rot=<0.7071 0.7071 0 0>}
 weapon_l {parent=hand_l, loc=<-0.1974 -3.4111 0.3811>, rot=<0 0 0 1>}
 lowerarm_twist_02_l {parent=lowerarm_l, loc=<10.6707 0 0>, rot=<0 0 0 1>}
 lowerarm_twist_01_l {parent=lowerarm_l, loc=<18.316 0 0>, rot=<0 0 0 1>}
 deform_elbow_in_base_l {parent=lowerarm_l, loc=<0.1834 3.6062 -0.2572>, rot=<0 0 0 -1>}
 deform_elbow_in_l {parent=deform_elbow_in_base_l, loc=<0 0 0>, rot=<0 0 0 1>}
 deform_elbow_out_base_l {parent=lowerarm_l, loc=<0 -4 0>, rot=<0 0 0 1>}
 deform_elbow_out_l {parent=deform_elbow_out_base_l, loc=<0 0 0>, rot=<0 0 0 1>}
 hand_l_null {parent=lowerarm_l, type=Null, loc=<22.7355 -0.3203 -0.6105>, rot=<-0.6784 -0.0223 0.0735 -0.7307>}
 upperarm_twist_01_l {parent=upperarm_l, loc=<7.2135 -0.1134 0>, rot=<0 0 0 1>}
 deform_upperarm_bk_base_l {parent=upperarm_twist_01_l, loc=<-6.4053 -4.3746 0.314>, rot=<0 0 0 -1>}
 deform_upperarm_bk_l {parent=deform_upperarm_bk_base_l, loc=<0 0 0>, rot=<0 0 0 1>}
 deform_upperarm_fr_base_l {parent=upperarm_twist_01_l, loc=<-5.6312 5.092 0.3354>, rot=<0 0 0 -1>}
 deform_upperarm_fr_l {parent=deform_upperarm_fr_base_l, loc=<0 0 0>, rot=<0 0 0 1>}
 deform_upperarm_up_base_l {parent=upperarm_twist_01_l, loc=<-8.891 0.4825 3.9517>, rot=<0 0 0 -1>}
 deform_upperarm_up_l {parent=deform_upperarm_up_base_l, loc=<0 0 0>, rot=<0 0 0 1>}
 base_upperarm_l_sleeves_bk {parent=upperarm_twist_01_l, loc=<-4.9398 -4.9632 -2.7248>, rot=<0 0 0 1>}
 dyn_upperarm_l_sleeves_bk {parent=base_upperarm_l_sleeves_bk, loc=<0 0 0>, rot=<0 0 0 1>}
 dyn_upperarm_l_sleeves_bk_end {parent=dyn_upperarm_l_sleeves_bk, loc=<7.2483 -1.8714 0.3374>, rot=<0 0 0 1>}
 base_upperarm_l_sleeves_fw {parent=upperarm_twist_01_l, loc=<-4.2986 5.7438 -1.9463>, rot=<0 0 0 1>}
 dyn_upperarm_l_sleeves_fw {parent=base_upperarm_l_sleeves_fw, loc=<0 0 0>, rot=<0 0 0 1>}
 dyn_upperarm_l_sleeves_fw_end {parent=dyn_upperarm_l_sleeves_fw, loc=<6.4887 0.9672 0.5886>, rot=<0 0 0 1>}
 base_upperarm_l_sleeves_up {parent=upperarm_twist_01_l, loc=<-7.6249 0.2436 4.1145>, rot=<0 0 0 1>}
 dyn_upperarm_l_sleeves_up {parent=base_upperarm_l_sleeves_up, loc=<0 0 0>, rot=<0 0 0 1>}
 dyn_upperarm_l_sleeves_up_end {parent=dyn_upperarm_l_sleeves_up, loc=<10.9432 -0.8212 1.3469>, rot=<0 0 0 1>}
 base_upperarm_l_sleeves_dw {parent=upperarm_twist_01_l, loc=<-1.7463 0.2436 -6.9189>, rot=<0 0 0 1>}
 dyn_upperarm_l_sleeves_dw {parent=base_upperarm_l_sleeves_dw, loc=<0 0 0>, rot=<0 0 0 1>}
 dyn_upperarm_l_sleeves_dw_end {parent=dyn_upperarm_l_sleeves_dw, loc=<3.8429 -0.8212 0>, rot=<0 0 0 1>}
 upperarm_twist_02_l {parent=upperarm_l, loc=<19.7998 -0.1134 0>, rot=<0 0 0 1>}
 upperarm_l_null {parent=clavicle_l, type=Null, loc=<9.8658 -1.2136 -0.4105>, rot=<0.0164 -0.3245 -0.0269 -0.9454>}
 clavicle_r {parent=spine_05, loc=<5.4831 -0.1053 3.2946>, rot=<0.7289 0.0465 -0.6829 0.009>}
 upperarm_r {parent=clavicle_r, loc=<-9.8658 1.2136 0.4105>, rot=<0.0164 -0.3245 -0.0269 -0.9454>}
 lowerarm_r {parent=upperarm_r, loc=<-26.465 0.8345 -0.259>, rot=<0.0036 0.0161 -0.2298 -0.9731>}
 hand_r {parent=lowerarm_r, loc=<-22.7355 0.3203 0.6106>, rot=<-0.6784 -0.0223 0.0735 -0.7307>}
 thumb_01_r {parent=hand_r, loc=<-1.3118 0.3191 2.459>, rot=<0.6546 -0.4042 0.0664 -0.6354>}
 thumb_02_r {parent=thumb_01_r, loc=<-5.9336 0.0448 -0.0401>, rot=<-0.0037 0.0213 0.298 -0.9543>}
 thumb_03_r {parent=thumb_02_r, loc=<-2.2223 -0.3682 0.0907>, rot=<0.0118 -0.013 0.0066 -0.9998>}
 index_metacarpal_r {parent=hand_r, loc=<-1.6535 -0.0183 2.96>, rot=<0.0158 -0.0124 -0.0173 -0.9996>}
 index_01_r {parent=index_metacarpal_r, loc=<-8.3008 -0.0357 0.0524>, rot=<-0.0044 0.0485 0.0924 -0.9945>}
 index_02_r {parent=index_01_r, loc=<-3.3587 0.3004 -0.0452>, rot=<-0.002 0.0027 0.3225 -0.9466>}
 index_03_r {parent=index_02_r, loc=<-2.5578 0.123 0.0252>, rot=<-0.002 0.0067 0.2558 -0.9667>}
 middle_metacarpal_r {parent=hand_r, loc=<-2.1856 0.1725 0.5984>, rot=<0.006 0.0168 -0.0329 -0.9993>}
 middle_01_r {parent=middle_metacarpal_r, loc=<-7.3275 0.2447 -0.0129>, rot=<-0.0208 0.0042 0.0858 -0.9961>}
 middle_02_r {parent=middle_01_r, loc=<-4.0123 -0.1401 -0.1175>, rot=<-0.0218 0.0258 0.3512 -0.9357>}
 middle_03_r {parent=middle_02_r, loc=<-2.6077 -0.0534 -0.1373>, rot=<0.0082 0.0188 0.1708 -0.9851>}
 ring_metacarpal_r {parent=hand_r, loc=<-2.2074 0.3625 -1.3638>, rot=<-0.0284 0.0722 -0.0458 -0.9959>}
 ring_01_r {parent=ring_metacarpal_r, loc=<-7.0295 0.6197 0.1446>, rot=<-0.0489 -0.0252 0.1892 -0.9804>}
 ring_02_r {parent=ring_01_r, loc=<-3.6174 -0.4647 0.1027>, rot=<0 0 0.2806 -0.9598>}
 ring_03_r {parent=ring_02_r, loc=<-2.4152 -0.114 -0.0199>, rot=<0 0 0.1189 -0.9929>}
 pinky_metacarpal_r {parent=hand_r, loc=<-1.7616 0.9753 -3.3814>, rot=<-0.0397 0.0814 -0.069 -0.9935>}
 pinky_01_r {parent=pinky_metacarpal_r, loc=<-7.1709 0.4831 0.1094>, rot=<-0.0654 -0.0035 0.2242 -0.9723>}
 pinky_02_r {parent=pinky_01_r, loc=<-2.8274 -0.0893 0.1043>, rot=<0.0037 0.0011 0.2854 -0.9584>}
 pinky_03_r {parent=pinky_02_r, loc=<-2.2046 -0.0307 0.0078>, rot=<-0.0002 0.0098 0.1445 -0.9895>}
 deform_wrist_dn_base_r {parent=hand_r, loc=<-0.0001 1.5328 0>, rot=<0 0 0 -1>}
 deform_wrist_dn_r {parent=deform_wrist_dn_base_r, loc=<0 0 0>, rot=<0 0 0 1>}
 deform_wrist_up_base_r {parent=hand_r, loc=<-0.0002 -2.3007 -0.0001>, rot=<0 0 0 -1>}
 deform_wrist_up_r {parent=deform_wrist_up_base_r, loc=<0 0 0>, rot=<0 0 0 1>}
 weapon_r {parent=hand_r, loc=<0.1974 3.4111 -0.3811>, rot=<0 0 0 1>}
 hand_attach_r {parent=hand_r, loc=<-8.5001 2.9999 0>, rot=<0 0 -0.7071 -0.7071>}
 lowerarm_twist_02_r {parent=lowerarm_r, loc=<-10.6707 0 0>, rot=<0 0 0 1>}
 lowerarm_twist_01_r {parent=lowerarm_r, loc=<-18.316 0 0>, rot=<0 0 0 1>}
 deform_elbow_in_base_r {parent=lowerarm_r, loc=<-0.1834 -3.6062 0.2572>, rot=<0 0 0 -1>}
 deform_elbow_in_r {parent=deform_elbow_in_base_r, loc=<0 0 0>, rot=<0 0 0 1>}
 deform_elbow_out_base_r {parent=lowerarm_r, loc=<0 4 -0.0001>, rot=<0 0 0 -1>}
 deform_elbow_out_r {parent=deform_elbow_out_base_r, loc=<0 0 0>, rot=<0 0 0 1>}
 hand_r_null {parent=lowerarm_r, type=Null, loc=<-22.7355 0.3203 0.6106>, rot=<-0.6784 -0.0223 0.0735 -0.7307>}
 upperarm_twist_01_r {parent=upperarm_r, loc=<-7.2135 0.1134 0>, rot=<0 0 0 -1>}
 deform_upperarm_bk_base_r {parent=upperarm_twist_01_r, loc=<6.4053 4.3746 -0.314>, rot=<0 0 0 -1>}
 deform_upperarm_bk_r {parent=deform_upperarm_bk_base_r, loc=<0 0 0>, rot=<0 0 0 1>}
 deform_upperarm_fr_base_r {parent=upperarm_twist_01_r, loc=<5.6312 -5.092 -0.3354>, rot=<0 0 0 -1>}
 deform_upperarm_fr_r {parent=deform_upperarm_fr_base_r, loc=<0 0 0>, rot=<0 0 0 1>}
 deform_upperarm_up_base_r {parent=upperarm_twist_01_r, loc=<8.891 -0.4825 -3.9517>, rot=<0 0 0 -1>}
 deform_upperarm_up_r {parent=deform_upperarm_up_base_r, loc=<0 0 0>, rot=<0 0 0 1>}
 base_upperarm_r_sleeves_bk {parent=upperarm_twist_01_r, loc=<4.9399 4.9632 2.7247>, rot=<0 -1 0 0>}
 dyn_upperarm_r_sleeves_bk {parent=base_upperarm_r_sleeves_bk, loc=<0 0 0>, rot=<0 0 0 1>}
 dyn_upperarm_r_sleeves_bk_end {parent=dyn_upperarm_r_sleeves_bk, loc=<7.2486 1.8714 0.3369>, rot=<0 0 0 1>}
 base_upperarm_r_sleeves_fw {parent=upperarm_twist_01_r, loc=<4.2984 -5.7438 1.9467>, rot=<0 -1 0 0>}
 dyn_upperarm_r_sleeves_fw {parent=base_upperarm_r_sleeves_fw, loc=<0 0 0>, rot=<0 0 0 1>}
 dyn_upperarm_r_sleeves_fw_end {parent=dyn_upperarm_r_sleeves_fw, loc=<6.4887 -0.9672 0.5886>, rot=<0 0 0 1>}
 base_upperarm_r_sleeves_up {parent=upperarm_twist_01_r, loc=<7.6252 -0.2435 -4.1148>, rot=<0 -1 0 0>}
 dyn_upperarm_r_sleeves_up {parent=base_upperarm_r_sleeves_up, loc=<0 0 0>, rot=<0 0 0 1>}
 dyn_upperarm_r_sleeves_up_end {parent=dyn_upperarm_r_sleeves_up, loc=<10.9433 0.8212 1.3468>, rot=<0 0 0 1>}
 base_upperarm_r_sleeves_dw {parent=upperarm_twist_01_r, loc=<1.746 -0.2436 6.9192>, rot=<0 -1 0 0>}
 dyn_upperarm_r_sleeves_dw {parent=base_upperarm_r_sleeves_dw, loc=<0 0 0>, rot=<0 0 0 1>}
 dyn_upperarm_r_sleeves_dw_end {parent=dyn_upperarm_r_sleeves_dw, loc=<3.8427 0.8212 0.0003>, rot=<0 0 0 1>}
 upperarm_twist_02_r {parent=upperarm_r, loc=<-19.7998 0.1134 0>, rot=<0 0 0 -1>}
 upperarm_r_null {parent=clavicle_r, type=Null, loc=<-9.8658 1.2136 0.4105>, rot=<0.0164 -0.3245 -0.0269 -0.9454>}
 deform_pec_aim_base_l {parent=spine_04, loc=<3.2456 9.3309 -1.2237>, rot=<0.0603 -0.4033 0.3318 -0.8507>}
 deform_pec_aim_l {parent=deform_pec_aim_base_l, loc=<0 0 0>, rot=<0 0 0 1>}
 deform_pec_base_l {parent=deform_pec_aim_l, loc=<9.3665 0 0>, rot=<0 0 0 1>}
 deform_pec_l {parent=deform_pec_base_l, loc=<0 0 0>, rot=<0 0 0 1>}
 deform_pec_aim_base_r {parent=spine_04, loc=<3.2452 9.3309 1.2237>, rot=<-0.4033 -0.0603 0.8507 0.3318>}
 deform_pec_aim_r {parent=deform_pec_aim_base_r, loc=<0 0 0>, rot=<0 0 0 1>}
 deform_pec_base_r {parent=deform_pec_aim_r, loc=<-9.3667 -0.0003 -0.0004>, rot=<0 0 0 1>}
 deform_pec_r {parent=deform_pec_base_r, loc=<0 0 0>, rot=<0 0 0 1>}
 thigh_l {parent=pelvis, loc=<-1.0855 -0.2109 -8.7053>, rot=<0.0009 0.0355 0.0155 -0.9993>}
 calf_l {parent=thigh_l, loc=<-42.5284 -0.0019 -0.7297>, rot=<0.0006 -0.0161 -0.0391 -0.9991>}
 foot_l {parent=calf_l, loc=<-46.7401 -2.5014 -0.592>, rot=<0.0427 -0.0238 0.0353 -0.9982>}
 ball_l {parent=foot_l, loc=<-3.5658 10.8066 -0.1544>, rot=<0.0001 -0.0001 -0.7063 -0.7079>}
 calf_twist_02_l {parent=calf_l, loc=<-19.1546 -1.7943 -0.2638>, rot=<0.0237 0.0095 -0.0035 -0.9997>}
 calf_twist_01_l {parent=calf_l, loc=<-35.8483 -1.9197 -0.5794>, rot=<0.0237 0.0095 -0.0035 -0.9997>}
 deform_knee_in_base_l {parent=calf_l, loc=<0.8031 -5.0231 0.2456>, rot=<0 0 0 -1>}
 deform_knee_in_l {parent=deform_knee_in_base_l, loc=<0 0 0>, rot=<0 0 0 1>}
 deform_knee_out_base_l {parent=calf_l, loc=<0 4 0>, rot=<0 0 0 -1>}
 deform_knee_out_l {parent=deform_knee_out_base_l, loc=<0 0 0>, rot=<0 0 0 1>}
 thigh_twist_01_l {parent=thigh_l, loc=<-27.952 0 0>, rot=<0.0013 0 0 -1>}
 deform_groin_base_l {parent=thigh_l, loc=<0.4801 5.6776 -0.2486>, rot=<0 0 0 -1>}
 deform_groin_l {parent=deform_groin_base_l, loc=<0 0 0>, rot=<0 0 0 1>}
 calf_l_null {parent=thigh_l, type=Null, loc=<-42.5284 -0.0019 -0.7297>, rot=<0.0006 -0.0161 -0.0391 -0.9991>}
 thigh_r {parent=pelvis, loc=<-1.0855 -0.2109 8.7053>, rot=<0.0355 -0.0009 0.9993 0.0155>}
 calf_r {parent=thigh_r, loc=<42.5284 0.0019 0.7297>, rot=<0.0006 -0.0161 -0.0391 -0.9991>}
 foot_r {parent=calf_r, loc=<46.7401 2.5014 0.592>, rot=<0.0427 -0.0238 0.0353 -0.9982>}
 ball_r {parent=foot_r, loc=<3.5658 -10.8066 0.1544>, rot=<0.0001 -0.0001 -0.7063 -0.7079>}
 calf_twist_02_r {parent=calf_r, loc=<19.1546 1.7943 0.2638>, rot=<0.0237 0.0095 -0.0035 -0.9997>}
 calf_twist_01_r {parent=calf_r, loc=<35.8483 1.9197 0.5794>, rot=<0.0237 0.0095 -0.0035 -0.9997>}
 deform_knee_in_base_r {parent=calf_r, loc=<-0.8031 5.0231 -0.2457>, rot=<0 0 0 -1>}
 deform_knee_in_r {parent=deform_knee_in_base_r, loc=<0 0 0>, rot=<0 0 0 1>}
 deform_knee_out_base_r {parent=calf_r, loc=<0 -4 0>, rot=<0 0 0 -1>}
 deform_knee_out_r {parent=deform_knee_out_base_r, loc=<0 0 0>, rot=<0 0 0 1>}
 thigh_twist_01_r {parent=thigh_r, loc=<27.952 0 0>, rot=<0.0013 0 0 -1>}
 deform_groin_base_r {parent=thigh_r, loc=<-0.4801 -5.6776 0.2486>, rot=<0 0 0 -1>}
 deform_groin_r {parent=deform_groin_base_r, loc=<0 0 0>, rot=<0 0 0 1>}
 calf_r_null {parent=thigh_r, type=Null, loc=<42.5284 0.0019 0.7297>, rot=<0.0006 -0.0161 -0.0391 -0.9991>}
 deform_glute_base_l {parent=pelvis, loc=<-7.3466 -9.0213 -9.1245>, rot=<0.0172 0.031 -0.4477 -0.8935>}
 deform_glute_l {parent=deform_glute_base_l, loc=<0 0 0>, rot=<0 0 0 1>}
 deform_glute_base_r {parent=pelvis, loc=<-7.3464 -9.0213 9.1245>, rot=<0.031 -0.0172 0.8935 -0.4477>}
 deform_glute_r {parent=deform_glute_base_r, loc=<0 0 0>, rot=<0 0 0 1>}
 thigh_l_null {parent=pelvis, type=Null, loc=<-1.0855 -0.2109 -8.7053>, rot=<0.0009 0.0355 0.0155 -0.9993>}
 thigh_r_null {parent=pelvis, type=Null, loc=<-1.0855 -0.2109 8.7053>, rot=<0.0355 -0.0009 0.9993 0.0155>}
 base_L_skirt_a_1 {parent=pelvis, loc=<-0.9136 7.0322 -10.7183>, rot=<-0.0147 -0.819 -0.5735 -0.0103>}
 base_L_skirt_a_2 {parent=base_L_skirt_a_1, loc=<16.3581 0 0>, rot=<0 0 0 1>}
 dyn_L_skirt_a_2 {parent=base_L_skirt_a_2, loc=<0 0 0>, rot=<0 0 0 1>}
 dyn_L_skirt_a_3 {parent=dyn_L_skirt_a_2, loc=<16.5916 0 0>, rot=<0 0 0 1>}
 dyn_L_skirt_a_3_end {parent=dyn_L_skirt_a_3, loc=<12.9541 0 0>, rot=<0 0 0 1>}
 base_L_skirt_b_1 {parent=pelvis, loc=<-1.2434 -2.168 -16.2648>, rot=<-0.0179 -0.9998 0 0>}
 base_L_skirt_b_2 {parent=base_L_skirt_b_1, loc=<16.3581 0 0>, rot=<0 0 0 1>}
 dyn_L_skirt_b_2 {parent=base_L_skirt_b_2, loc=<0 0 0>, rot=<0 0 0 1>}
 dyn_L_skirt_b_3 {parent=dyn_L_skirt_b_2, loc=<16.5916 0 0>, rot=<0 0 0 1>}
 dyn_L_skirt_b_3_end {parent=dyn_L_skirt_b_3, loc=<12.9541 0 0>, rot=<0 0 0 1>}
 base_L_skirt_c_1 {parent=pelvis, loc=<-1.5459 -10.6077 -9.3728>, rot=<-0.0036 -0.8191 0.5736 0.0025>}
 base_L_skirt_c_2 {parent=base_L_skirt_c_1, loc=<16.3581 0 0>, rot=<0 0 0 1>}
 dyn_L_skirt_c_2 {parent=base_L_skirt_c_2, loc=<0 0 0>, rot=<0 0 0 1>}
 dyn_L_skirt_c_3 {parent=dyn_L_skirt_c_2, loc=<16.5916 0 0>, rot=<0 0 0 1>}
 dyn_L_skirt_c_3_end {parent=dyn_L_skirt_c_3, loc=<12.9541 0 0>, rot=<0 0 0 1>}
 base_R_skirt_a_1 {parent=pelvis, loc=<-0.9136 7.0322 10.7183>, rot=<0.0103 0.5735 0.819 0.0147>}
 base_R_skirt_a_2 {parent=base_R_skirt_a_1, loc=<16.3581 0 0>, rot=<0 0 0 1>}
 dyn_R_skirt_a_2 {parent=base_R_skirt_a_2, loc=<0 0 0>, rot=<0 0 0 -1>}
 dyn_R_skirt_a_3 {parent=dyn_R_skirt_a_2, loc=<16.5916 0 0>, rot=<0 0 0 -1>}
 dyn_R_skirt_a_3_end {parent=dyn_R_skirt_a_3, loc=<12.954 0 0>, rot=<0 0 0 -1>}
 base_R_skirt_b_1 {parent=pelvis, loc=<-1.2434 -2.168 16.2648>, rot=<0 0 0.9998 0.0179>}
 base_R_skirt_b_2 {parent=base_R_skirt_b_1, loc=<16.3581 0 0>, rot=<0 0 0 1>}
 dyn_R_skirt_b_2 {parent=base_R_skirt_b_2, loc=<0 0 0>, rot=<0 0 0 1>}
 dyn_R_skirt_b_3 {parent=dyn_R_skirt_b_2, loc=<16.5916 0 0>, rot=<0 0 0 1>}
 dyn_R_skirt_b_3_end {parent=dyn_R_skirt_b_3, loc=<12.954 0 0>, rot=<0 0 0 1>}
 base_R_skirt_c_1 {parent=pelvis, loc=<-1.5459 -10.6077 9.3728>, rot=<-0.0025 -0.5736 0.8191 0.0036>}
 base_R_skirt_c_2 {parent=base_R_skirt_c_1, loc=<16.3581 0 0>, rot=<0 0 0 1>}
 dyn_R_skirt_c_2 {parent=base_R_skirt_c_2, loc=<0 0 0>, rot=<0 0 0 1>}
 dyn_R_skirt_c_3 {parent=dyn_R_skirt_c_2, loc=<16.5916 0 0>, rot=<0 0 0 1>}
 dyn_R_skirt_c_3_end {parent=dyn_R_skirt_c_3, loc=<12.9541 0 0>, rot=<0 0 0 1>}
 base_C_skirt_a_1 {parent=pelvis, loc=<-0.8659 8.3629 0.0001>, rot=<0.0127 0.707 0.707 0.0127>}
 base_C_skirt_a_2 {parent=base_C_skirt_a_1, loc=<16.3581 0 0>, rot=<0 0 0 1>}
 dyn_C_skirt_a_2 {parent=base_C_skirt_a_2, loc=<0 0 0>, rot=<0 0 0 1>}
 dyn_C_skirt_a_3 {parent=dyn_C_skirt_a_2, loc=<16.5916 0 0>, rot=<0 0 0 1>}
 dyn_C_skirt_a_3_end {parent=dyn_C_skirt_a_3, loc=<12.9541 0 0>, rot=<0 0 0 1>}
 base_C_skirt_b_1 {parent=pelvis, loc=<-1.61 -12.396 0.0001>, rot=<-0.0031 -0.7071 0.7071 0.0031>}
 base_C_skirt_b_2 {parent=base_C_skirt_b_1, loc=<16.3581 0 0>, rot=<0 0 0 1>}
 dyn_C_skirt_b_2 {parent=base_C_skirt_b_2, loc=<0 0 0>, rot=<0 0 0 1>}
 dyn_C_skirt_b_3 {parent=dyn_C_skirt_b_2, loc=<16.5916 0 0>, rot=<0 0 0 1>}
 dyn_C_skirt_b_3_end {parent=dyn_C_skirt_b_3, loc=<12.9541 0 0>, rot=<0 0 0 1>}
 floor {parent=root, loc=<0 0 0>, rot=<0 0 0 1>}
 attach {parent=root, loc=<0 0 0>, rot=<0 0 0 1>}
 ik_hand_root {parent=root, loc=<0 0 0>, rot=<0 0 0 1>}
 ik_hand_gun {parent=ik_hand_root, loc=<-48.1743 3.3545 104.4381>, rot=<0.6276 0.1309 0.4183 -0.6434>}
 ik_hand_r {parent=ik_hand_gun, loc=<0 0 0>, rot=<0 0 0 1>}
 ik_hand_l {parent=ik_hand_gun, loc=<59.3271 67.6997 34.3524>, rot=<-0.6158 -0.7027 -0.3565 0>}
 ik_foot_root {parent=root, loc=<0 0 0>, rot=<0 0 0 1>}
 ik_foot_l {parent=ik_foot_root, loc=<14.8262 -7.3545 8.296>, rot=<0.0357 0.7045 0.0284 -0.7082>}
 ik_foot_r {parent=ik_foot_root, loc=<-14.8262 -7.3545 8.296>, rot=<0.7082 0.0284 -0.7045 0.0357>}
 root_fp {parent=root, loc=<0 0 0>, rot=<0 0 0 1>}
 attach_fp {parent=root_fp, loc=<0 0 0>, rot=<0 0 0 1>}
 camera_root {parent=root_fp, loc=<0 0 0>, rot=<0 0 0 1>}
 camera {parent=camera_root, loc=<0 0 0>, rot=<0 0 0 1>}
 attach_emote_01 {parent=root, loc=<0 0 0>, rot=<0 0 0 1>}
 attach_emote_02 {parent=root, loc=<0 0 0>, rot=<0 0 0 1>}
 attach_emote_03 {parent=root, loc=<0 0 0>, rot=<0 0 0 1>}
 warp_point_01 {parent=root, loc=<0 0 0>, rot=<0 0 0 1>}
 warp_point_02 {parent=root, loc=<0 0 0>, rot=<0 0 0 1>}
 warp_point_03 {parent=root, loc=<0 0 0>, rot=<0 0 0 1>}
 pelvis_ctrl {controlType=EULER_TRANSFORM, type=Control, loc=<0 0 0>, rot=<0 0 0 1>}
 thigh_l_ctrl {controlType=EULER_TRANSFORM, parent=pelvis_ctrl, type=Control, loc=<0 0 0>, rot=<0 0 0 1>}
 calf_l_ctrl {controlType=EULER_TRANSFORM, parent=thigh_l_ctrl, type=Control, loc=<0 0 0>, rot=<0 0 0 1>}
 thigh_r_ctrl {controlType=EULER_TRANSFORM, parent=pelvis_ctrl, type=Control, loc=<0 0 0>, rot=<0 0 0 1>}
 calf_r_ctrl {controlType=EULER_TRANSFORM, parent=thigh_r_ctrl, type=Control, loc=<0 0 0>, rot=<0 0 0 1>}
 spine_01_ctrl {controlType=EULER_TRANSFORM, parent=pelvis_ctrl, type=Control, loc=<0 0 0>, rot=<0 0 0 1>}
 spine_02_ctrl {controlType=EULER_TRANSFORM, parent=spine_01_ctrl, type=Control, loc=<0 0 0>, rot=<0 0 0 1>}
 spine_03_ctrl {controlType=EULER_TRANSFORM, parent=spine_02_ctrl, type=Control, loc=<0 0 0>, rot=<0 0 0 1>}
 spine_04_ctrl {controlType=EULER_TRANSFORM, parent=spine_03_ctrl, type=Control, loc=<0 0 0>, rot=<0 0 0 1>}
 spine_05_ctrl {controlType=EULER_TRANSFORM, parent=spine_04_ctrl, type=Control, loc=<0 0 0>, rot=<0 0 0 1>}
 clavicle_l_ctrl {controlType=EULER_TRANSFORM, parent=spine_05_ctrl, type=Control, loc=<0 0 0>, rot=<0 0 0 1>}
 upperarm_l_ctrl {controlType=EULER_TRANSFORM, parent=clavicle_l_ctrl, type=Control, loc=<0 0 0>, rot=<0 0 0 1>}
 lowerarm_l_ctrl {controlType=EULER_TRANSFORM, parent=upperarm_l_ctrl, type=Control, loc=<0 0 0>, rot=<0 0 0 1>}
 hand_l_ctrl {controlType=EULER_TRANSFORM, parent=lowerarm_l_ctrl, type=Control, loc=<0 0 0>, rot=<0 0 0 1>}
 clavicle_r_ctrl {controlType=EULER_TRANSFORM, parent=spine_05_ctrl, type=Control, loc=<0 0 0>, rot=<0 0 0 1>}
 upperarm_r_ctrl {controlType=EULER_TRANSFORM, parent=clavicle_r_ctrl, type=Control, loc=<0 0 0>, rot=<0 0 0 1>}
 lowerarm_r_ctrl {controlType=EULER_TRANSFORM, parent=upperarm_r_ctrl, type=Control, loc=<0 0 0>, rot=<0 0 0 1>}
 hand_r_ctrl {controlType=EULER_TRANSFORM, parent=lowerarm_r_ctrl, type=Control, loc=<0 0 0>, rot=<0 0 0 1>}
]
variables [
 l leg backwards SPR float {type=float, default=0}
 l thighLocal ry float {type=double, default=0}
 l thighLocal rz float {type=double, default=0}
 pub disablecontrolRig {type=bool, default=True}
 pub isBaseballSliding {type=double, default=0}
 pub isCrouchStrafeLeft {type=double, default=0}
 pub isCrouchStrafeRight {type=double, default=0}
 pub isCrouching {type=bool, default=False}
 "pub isDBNO " {type=bool, default=False}
 pub isFrontEnd {type=bool, default=False}
 pub isGrindRail {type=double, default=0}
 pub isKneeSliding {type=double, default=0}
 pub isPlayingEasternBloc {type=bool, default=False}
 pub isPlayingEmote {type=bool, default=False}
 pub isZipline {type=double, default=0}
 r leg backwards SPR float {type=float, default=0}
 r thighLocal ry float {type=double, default=0}
 r thighLocal rz float {type=double, default=0}
]
