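# Physics asset for the F_MED_HayBrush female player cosmetic body, set up for a dynamic skirt around the legs.
# Eight skirt chains (two center, three per side) each link a tiny base anchor capsule to two dyn segments.
# Swing limits are tighter at the root than the lower joint, and only center chains allow twist.
# Thigh and calf capsules supply leg shapes, joined by knee constraints with free rotation.
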PhysicsAsset /BRCosmetics/Characters/Player/Female/Medium/Bodies/F_MED_HayBrush/Meshes/F_MED_HayBrush_Physics
bodies {
 base_C_skirt_a_2 [{shapeType=Capsule, length=0.505, radius=0.505, shapeName=base_C_skirt_a_2_capsule, rotation=<rot 0 -90 0>}]
 base_C_skirt_b_2 [{shapeType=Capsule, length=0.505, radius=0.505, shapeName=base_C_skirt_b_2_capsule, rotation=<rot 0 -90 0>}]
 base_L_skirt_a_2 [{shapeType=Capsule, length=0.505, radius=0.505, shapeName=base_L_skirt_a_2_capsule, rotation=<rot 0 -90 0>}]
 base_L_skirt_b_2 [{shapeType=Capsule, length=0.505, radius=0.505, shapeName=base_L_skirt_b_2_capsule, rotation=<rot 0 -90 0>}]
 base_L_skirt_c_2 [{shapeType=Capsule, length=0.505, radius=0.505, shapeName=base_L_skirt_c_2_capsule, rotation=<rot 0 -90 0>}]
 base_R_skirt_a_2 [{shapeType=Capsule, length=0.505, radius=0.505, shapeName=base_R_skirt_a_2_capsule, rotation=<rot 0 -90 0>}]
 base_R_skirt_b_2 [{shapeType=Capsule, length=0.505, radius=0.505, shapeName=base_R_skirt_b_2_capsule, rotation=<rot 0 -90 0>}]
 base_R_skirt_c_2 [{shapeType=Capsule, length=0.505, radius=0.505, shapeName=base_R_skirt_c_2_capsule, rotation=<rot 0 -90 0>}]
 calf_l [
  {shapeType=Capsule, length=16.9569, radius=5.1427, shapeName=calf_l_capsule, center=<-9.8209 -1.7483 0.1458>, rotation=<rot 88 180 170>}
  {
   shapeType=Capsule
   length=33.5341
   radius=7
   shapeName=calf_l_capsule
   center=<-30.6082 -3.0424 -0.881>
   rotation=<rot 89.5458 -172.0139 -169.6696>
  }
 ]
 calf_r [
  {shapeType=Capsule, length=16.9569, radius=5.1427, shapeName=calf_l_capsule, center=<9.8209 1.7483 -0.1458>, rotation=<rot 88 180 -10>}
  {
   shapeType=Capsule
   length=33.5341
   radius=7
   shapeName=calf_l_capsule
   center=<30.6082 3.0424 0.881>
   rotation=<rot 89.5458 -172.0139 10.3304>
  }
 ]
 dyn_C_skirt_a_2 [{shapeType=Capsule, length=11.1913, radius=4.8087, shapeName=dyn_C_skirt_a_2_capsule, center=<8.7138 0 2.795>, rotation=<rot 90 0 0>}]
 dyn_C_skirt_a_3 [{shapeType=Capsule, length=10.9732, radius=4, shapeName=dyn_C_skirt_a_2_capsule, center=<10.409 0 2.3962>, rotation=<rot 90 0 0>}]
 dyn_C_skirt_b_2 [{shapeType=Capsule, length=11.1913, radius=4.8087, shapeName=dyn_C_skirt_a_2_capsule, center=<8.7138 0 2.795>, rotation=<rot 90 0 0>}]
 dyn_C_skirt_b_3 [{shapeType=Capsule, length=10.9732, radius=4, shapeName=dyn_C_skirt_a_2_capsule, center=<10.409 0 2.3962>, rotation=<rot 90 0 0>}]
 dyn_L_skirt_a_2 [
  {shapeType=Capsule, length=12, radius=4, shapeName=dyn_C_skirt_a_2_capsule, center=<7.5 -5.8963 2.462>, rotation=<rot -90 180 180>}
  {shapeType=Capsule, length=12, radius=4, shapeName=dyn_C_skirt_a_2_capsule, center=<7.5 2.4288 4.2716>, rotation=<rot -90 0 0>}
 ]
 dyn_L_skirt_a_3 [
  {shapeType=Capsule, length=14.2439, radius=4, shapeName=dyn_C_skirt_a_2_capsule, center=<11.5302 0 1.7187>, rotation=<rot 0 0 -90>}
  {shapeType=Capsule, length=14.2439, radius=4, shapeName=dyn_R_skirt_a_3_capsule, center=<3.6096 0 1.3898>, rotation=<rot 0 0 -90>}
 ]
 dyn_L_skirt_b_2 [{shapeType=Capsule, length=12, radius=4, shapeName=dyn_C_skirt_a_2_capsule, center=<7.5 0 4.0686>, rotation=<rot 0 0 -90>}]
 dyn_L_skirt_b_3 [{shapeType=Capsule, length=12, radius=4, shapeName=dyn_C_skirt_a_2_capsule, center=<7.5 0 3.46>, rotation=<rot 0 0 -90>}]
 dyn_L_skirt_c_2 [
  {shapeType=Capsule, length=12, radius=4, shapeName=dyn_C_skirt_a_2_capsule, center=<7.5 5.091 2.0638>, rotation=<rot -90 0 0>}
  {shapeType=Capsule, length=12, radius=4, shapeName=dyn_C_skirt_a_2_capsule, center=<7.5 -3.2123 2.2308>, rotation=<rot -90 0 0>}
 ]
 dyn_L_skirt_c_3 [
  {shapeType=Capsule, length=14.2439, radius=4, shapeName=dyn_C_skirt_c_3_capsule, center=<5.2818 0 1.7338>, rotation=<rot 0 0 -90>}
  {shapeType=Capsule, length=14.2439, radius=4, shapeName=dyn_L_skirt_c_3_capsule, center=<14.3599 0 2.5243>, rotation=<rot 0 0 -90>}
 ]
 dyn_R_skirt_a_2 [
  {shapeType=Capsule, length=12, radius=4, shapeName=dyn_C_skirt_a_2_capsule, center=<7.5 -5.8963 2.8781>, rotation=<rot -90 180 180>}
  {shapeType=Capsule, length=12, radius=4, shapeName=dyn_C_skirt_a_2_capsule, center=<7.5 2.4288 3.2059>, rotation=<rot -90 0 0>}
 ]
 dyn_R_skirt_a_3 [
  {shapeType=Capsule, length=14.2439, radius=4, shapeName=dyn_C_skirt_a_2_capsule, center=<11.5302 0 1.7187>, rotation=<rot 0 0 -90>}
  {shapeType=Capsule, length=14.2439, radius=4, shapeName=dyn_R_skirt_a_3_capsule, center=<3.6096 0 1.3898>, rotation=<rot 0 0 -90>}
 ]
 dyn_R_skirt_b_2 [{shapeType=Capsule, length=12, radius=4, shapeName=dyn_C_skirt_a_2_capsule, center=<7.5 0 4.0686>, rotation=<rot 0 0 -90>}]
 dyn_R_skirt_b_3 [{shapeType=Capsule, length=12, radius=4, shapeName=dyn_C_skirt_a_2_capsule, center=<7.5 0 3.46>, rotation=<rot 0 0 -90>}]
 dyn_R_skirt_c_2 [
  {shapeType=Capsule, length=12, radius=4, shapeName=dyn_C_skirt_a_2_capsule, center=<7.5 4.019 2.0638>, rotation=<rot -90 0 0>}
  {shapeType=Capsule, length=12, radius=4, shapeName=dyn_C_skirt_a_2_capsule, center=<7.5 -5.2375 2.2308>, rotation=<rot -90 180 180>}
 ]
 dyn_R_skirt_c_3 [
  {shapeType=Capsule, length=14.2439, radius=4, shapeName=dyn_C_skirt_c_3_capsule, center=<5.2818 0 1.7338>, rotation=<rot 0 0 -90>}
  {shapeType=Capsule, length=14.2439, radius=4, shapeName=dyn_L_skirt_c_3_capsule, center=<14.3599 0 2.5243>, rotation=<rot 0 0 -90>}
 ]
 thigh_l [
  {
   shapeType=Capsule
   length=21.2415
   radius=6.9872
   shapeName=thigh_l_capsule
   center=<-13.2994 -0.7814 0.9974>
   rotation=<rot 88.8078 -5.1517 -2.3623>
  }
  {
   shapeType=Capsule
   length=16.9974
   radius=6.4131
   shapeName=thigh_l_capsule
   center=<-30.0564 0.0472 0.2459>
   rotation=<rot 85.5064 -2.9714 -1.564>
  }
 ]
 thigh_r [
  {
   shapeType=Capsule
   length=21.2415
   radius=6.9872
   shapeName=thigh_l_capsule
   center=<13.2994 0.7814 -0.9974>
   rotation=<rot 88.8078 -5.1517 177.6377>
  }
  {
   shapeType=Capsule
   length=16.9974
   radius=6.4131
   shapeName=thigh_l_capsule
   center=<30.0564 -0.0472 -0.2459>
   rotation=<rot 85.5064 -2.9714 178.436>
  }
 ]
}
constraints [
 {bone1Name=dyn_R_skirt_c_3, bone2Name=dyn_R_skirt_c_2, swing1LimitDegrees=15, swing2LimitDegrees=35, twistLimitDegrees=0}
 {bone1Name=dyn_R_skirt_b_3, bone2Name=dyn_R_skirt_b_2, swing1LimitDegrees=15, swing2LimitDegrees=35, twistLimitDegrees=0}
 {bone1Name=dyn_R_skirt_a_3, bone2Name=dyn_R_skirt_a_2, swing1LimitDegrees=15, swing2LimitDegrees=35, twistLimitDegrees=0}
 {bone1Name=dyn_L_skirt_c_3, bone2Name=dyn_L_skirt_c_2, swing1LimitDegrees=15, swing2LimitDegrees=35, twistLimitDegrees=0}
 {bone1Name=dyn_L_skirt_b_3, bone2Name=dyn_L_skirt_b_2, swing1LimitDegrees=15, swing2LimitDegrees=35, twistLimitDegrees=0}
 {bone1Name=dyn_L_skirt_a_3, bone2Name=dyn_L_skirt_a_2, swing1LimitDegrees=15, swing2LimitDegrees=35, twistLimitDegrees=0}
 {
  bone1Name=calf_l
  bone2Name=thigh_l
  swing1LimitDegrees=45
  swing1Motion=Free
  swing2LimitDegrees=45
  swing2Motion=Free
  twistLimitDegrees=45
  twistMotion=Free
 }
 {
  bone1Name=calf_r
  bone2Name=thigh_r
  swing1LimitDegrees=45
  swing1Motion=Free
  swing2LimitDegrees=45
  swing2Motion=Free
  twistLimitDegrees=45
  twistMotion=Free
 }
 {bone1Name=dyn_L_skirt_a_2, bone2Name=base_L_skirt_a_2, swing1LimitDegrees=5, swing2LimitDegrees=22, twistLimitDegrees=0}
 {bone1Name=dyn_L_skirt_b_2, bone2Name=base_L_skirt_b_2, swing1LimitDegrees=5, swing2LimitDegrees=22, twistLimitDegrees=0}
 {bone1Name=dyn_L_skirt_c_2, bone2Name=base_L_skirt_c_2, swing1LimitDegrees=5, swing2LimitDegrees=22, twistLimitDegrees=0}
 {bone1Name=dyn_R_skirt_a_2, bone2Name=base_R_skirt_a_2, swing1LimitDegrees=5, swing2LimitDegrees=22, twistLimitDegrees=0}
 {bone1Name=dyn_R_skirt_b_2, bone2Name=base_R_skirt_b_2, swing1LimitDegrees=5, swing2LimitDegrees=22, twistLimitDegrees=0}
 {bone1Name=dyn_R_skirt_c_2, bone2Name=base_R_skirt_c_2, swing1LimitDegrees=5, swing2LimitDegrees=22, twistLimitDegrees=0}
 {bone1Name=dyn_C_skirt_b_3, bone2Name=dyn_C_skirt_b_2, swing1LimitDegrees=15, swing2LimitDegrees=35, twistLimitDegrees=10}
 {bone1Name=dyn_C_skirt_b_2, bone2Name=base_C_skirt_b_2, swing1LimitDegrees=5, swing2LimitDegrees=22, twistLimitDegrees=15}
 {bone1Name=dyn_C_skirt_a_3, bone2Name=dyn_C_skirt_a_2, swing1LimitDegrees=15, swing2LimitDegrees=35, twistLimitDegrees=10}
 {bone1Name=dyn_C_skirt_a_2, bone2Name=base_C_skirt_a_2, swing1LimitDegrees=5, swing2LimitDegrees=22, twistLimitDegrees=15}
]
