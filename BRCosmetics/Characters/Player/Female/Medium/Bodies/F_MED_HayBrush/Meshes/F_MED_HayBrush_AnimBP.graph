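# Post-process style cosmetic AnimBP for the HayBrush female body, adding secondary motion to skirt and sleeves over the pose copied from the parent mesh.
# A state-driven Control Rig feeds sprint offsets, front-end or gameplay RigidBody skirt physics, then AnimDynamics sleeves.
# Additive bone fixes follow for skydiving, backpack, cape, tactical sprint and DBNO movement.
# The EventGraph derives movement alphas and gravity, plus sine-based rumble rotations for skydive flutter.
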AnimBlueprint /BRCosmetics/Characters/Player/Female/Medium/Bodies/F_MED_HayBrush/Meshes/F_MED_HayBrush_AnimBP
target_skeleton=&/F_MED_HayBrush_Skeleton
graphs [
 AnimGraph {nodes [
   RigidBody!(ComponentPose, Alpha, OverrideWorldGravity:Vector="0,0,0", ExternalForce) -> (Pose) {
    bOverrideWorldGravity=true
    simulationSpace=BaseBoneSpace
    alphaBoolBlend {blendOption=Linear}
    baseBoneRef {boneName=spine_05}
    componentLinearAccScale=<0.2 0.3 0.2>
    componentLinearVelScale=<0.2 0.3 0.2>
    simSpaceSettings {
     maxAngularAcceleration=600
     maxAngularVelocity=1000
     maxLinearAcceleration=500
     maxLinearVelocity=1000
     worldAlpha=0.1
     externalLinearDragV=<0.2 0.2>
    }
   }
   n1 = "Transform (Modify) Bone - Bone: base_upperarm_l_sleeves_bk"!(ComponentPose=Local To Component(LocalPose=Use cached pose 'DBNOSkirtAdjustmentsCP'()), Alpha=1, Rotation="0,0,0") -> (Pose) {
    rotationMode=BMM_Additive
    rotationSpace=BCS_BoneSpace
    alphaBoolBlend {blendOption=Linear}
    boneToModify {boneName=base_upperarm_l_sleeves_bk}
   }
   n2 = "Transform (Modify) Bone - Bone: base_upperarm_l_sleeves_fw"!(ComponentPose=@n1, Alpha=1, Rotation="0,0,0") -> (Pose) {
    rotationMode=BMM_Additive
    rotationSpace=BCS_BoneSpace
    alphaBoolBlend {blendOption=Linear}
    boneToModify {boneName=base_upperarm_l_sleeves_fw}
   }
   n3 = "Transform (Modify) Bone - Bone: base_upperarm_l_sleeves_up"!(ComponentPose=@n2, Alpha=1, Rotation="0,0,0") -> (Pose) {
    rotationMode=BMM_Additive
    rotationSpace=BCS_BoneSpace
    alphaBoolBlend {blendOption=Linear}
    boneToModify {boneName=base_upperarm_l_sleeves_up}
   }
   n4 = "Transform (Modify) Bone - Bone: base_upperarm_r_sleeves_bk"!(ComponentPose=@n3, Alpha=1, Rotation="0,0,0") -> (Pose) {
    rotationMode=BMM_Additive
    rotationSpace=BCS_BoneSpace
    alphaBoolBlend {blendOption=Linear}
    boneToModify {boneName=base_upperarm_r_sleeves_bk}
   }
   n5 = "Transform (Modify) Bone - Bone: base_upperarm_r_sleeves_fw"!(ComponentPose=@n4, Alpha=1, Rotation="0,0,0") -> (Pose) {
    rotationMode=BMM_Additive
    rotationSpace=BCS_BoneSpace
    alphaBoolBlend {blendOption=Linear}
    boneToModify {boneName=base_upperarm_r_sleeves_fw}
   }
   n6 = "Transform (Modify) Bone - Bone: base_upperarm_r_sleeves_up"!(ComponentPose=@n5, Alpha=1, Rotation="0,0,0") -> (Pose) {
    rotationMode=BMM_Additive
    rotationSpace=BCS_BoneSpace
    alphaBoolBlend {blendOption=Linear}
    boneToModify {boneName=base_upperarm_r_sleeves_up}
   }
   n7 = "Transform (Modify) Bone - Bone: dyn_L_skirt_a_3"!(ComponentPose=@n6, Alpha=1, Rotation="0,0,0") -> (Pose) {rotationMode=BMM_Additive, rotationSpace=BCS_BoneSpace, alphaBoolBlend {blendOption=Linear}, boneToModify {boneName=dyn_L_skirt_a_3}}
   n8 = "Transform (Modify) Bone - Bone: dyn_L_skirt_b_3"!(ComponentPose=@n7, Alpha=1, Rotation="0,0,0") -> (Pose) {rotationMode=BMM_Additive, rotationSpace=BCS_BoneSpace, alphaBoolBlend {blendOption=Linear}, boneToModify {boneName=dyn_L_skirt_b_3}}
   n9 = "Transform (Modify) Bone - Bone: dyn_L_skirt_c_3"!(ComponentPose=@n8, Alpha=1, Rotation="0,0,0") -> (Pose) {rotationMode=BMM_Additive, rotationSpace=BCS_BoneSpace, alphaBoolBlend {blendOption=Linear}, boneToModify {boneName=dyn_L_skirt_c_3}}
   n10 = "Transform (Modify) Bone - Bone: dyn_R_skirt_a_3"!(ComponentPose=@n9, Alpha=1, Rotation="0,0,0") -> (Pose) {rotationMode=BMM_Additive, rotationSpace=BCS_BoneSpace, alphaBoolBlend {blendOption=Linear}, boneToModify {boneName=dyn_R_skirt_a_3}}
   n11 = "Transform (Modify) Bone - Bone: dyn_R_skirt_b_3"!(ComponentPose=@n10, Alpha=1, Rotation="0,0,0") -> (Pose) {rotationMode=BMM_Additive, rotationSpace=BCS_BoneSpace, alphaBoolBlend {blendOption=Linear}, boneToModify {boneName=dyn_R_skirt_b_3}}
   n12 = "Transform (Modify) Bone - Bone: dyn_R_skirt_c_3"!(ComponentPose=@n11, Alpha=1, Rotation="0,0,0") -> (Pose) {rotationMode=BMM_Additive, rotationSpace=BCS_BoneSpace, alphaBoolBlend {blendOption=Linear}, boneToModify {boneName=dyn_R_skirt_c_3}}
   n13 = "Transform (Modify) Bone - Bone: dyn_C_skirt_a_3"!(ComponentPose=@n12, Alpha=1, Rotation="0,0,0") -> (Pose) {rotationMode=BMM_Additive, rotationSpace=BCS_BoneSpace, alphaBoolBlend {blendOption=Linear}, boneToModify {boneName=dyn_C_skirt_a_3}}
   n14 = "Transform (Modify) Bone - Bone: dyn_C_skirt_b_3"!(ComponentPose=@n13, Alpha=1, Rotation="0,0,0") -> (Pose) {rotationMode=BMM_Additive, rotationSpace=BCS_BoneSpace, alphaBoolBlend {blendOption=Linear}, boneToModify {boneName=dyn_C_skirt_b_3}}
   Output Pose(Result=Blend Poses by bool(
     BlendPose_0=Component To Local(ComponentPose=@n14)
     BlendPose_1=Use cached pose 'DBNOSkirtAdjustmentsCP'()
     bActiveValue=Get enableSin() -> (enableSin:bool)
    ) {blendType=Linear}) {layerGroup=None}
   n15 = Control Rig!(
    Source=Copy Pose From Mesh() {bUseAttachedParent=true}
    Alpha
    pub disablecontrolRig:bool=Get disableControlRig() -> (disableControlRig:bool)
    pub isKneeSliding:double=Get isKneeSlidling float() -> (isKneeSlidling float:double)
    pub isBaseballSliding:double=Get isBaseballSlidling float() -> (isBaseballSlidling float:double)
    pub isCrouchStrafeLeft:double=Get isCrouchStrafeLeft float() -> (isCrouchStrafeLeft float:double)
    pub isCrouchStrafeRight:double=Get isCrouchStrafeRight float() -> (isCrouchStrafeRight float:double)
    "pub isDBNO ":bool=Get bIsDBNO()
    pub isCrouching:bool=Get bIsCrouching()
    pub isZipline:double=Get isZipline float() -> (isZipline float:double)
    pub isGrindRail:double=Get isOnGrindRail float() -> (isOnGrindRail float:double)
    pub isPlayingEasternBloc:bool=Get isPlayingEasterBloc() -> (isPlayingEasterBloc:bool)
    pub isPlayingEmote:bool=Get bIsPlayingEmote()
    pub isFrontEnd:bool=Get bIsInFrontEnd()
   ) -> (Pose) {
    alphaBoolBlend {blendOption=Linear}
    controlRigAssetReference {blueprintRigClass=&/F_MED_HayBrush_CtrlRig.F_MED_HayBrush_CtrlRig_C}
   }
   Save cached pose 'ControlRigCP'(Pose=@n15) {cacheName=ControlRigCP}
   n16 = Get moveSkirtFwBk() -> (moveSkirtFwBk:double)
   n17 = "Transform (Modify) Bone - Bone: base_L_skirt_c_1"!(ComponentPose=Local To Component(LocalPose=Use cached pose 'ControlRigCP'()), Alpha=@n16, Translation="0,0,2", Rotation="30,0,0") -> (Pose) {
    rotationMode=BMM_Additive
    rotationSpace=BCS_BoneSpace
    translationMode=BMM_Additive
    translationSpace=BCS_BoneSpace
    alphaBoolBlend {blendOption=Linear}
    alphaScaleBiasClamp {bInterpResult=true}
    boneToModify {boneName=base_L_skirt_c_1}
    rotation {pitch=30}
    translation {z=2}
   }
   n18 = "Transform (Modify) Bone - Bone: base_C_skirt_b_1"!(ComponentPose=@n17, Alpha=@n16, Translation="0,0,2", Rotation="30,0,0") -> (Pose) {
    rotationMode=BMM_Additive
    rotationSpace=BCS_BoneSpace
    translationMode=BMM_Additive
    translationSpace=BCS_BoneSpace
    alphaBoolBlend {blendOption=Linear}
    alphaScaleBiasClamp {bInterpResult=true}
    boneToModify {boneName=base_C_skirt_b_1}
    rotation {pitch=30}
    translation {z=2}
   }
   n19 = "Transform (Modify) Bone - Bone: base_R_skirt_c_1"!(ComponentPose=@n18, Alpha=@n16, Translation="0,0,2", Rotation="30,0,0") -> (Pose) {
    rotationMode=BMM_Additive
    rotationSpace=BCS_BoneSpace
    translationMode=BMM_Additive
    translationSpace=BCS_BoneSpace
    alphaBoolBlend {blendOption=Linear}
    alphaScaleBiasClamp {bInterpResult=true}
    boneToModify {boneName=base_R_skirt_c_1}
    rotation {pitch=30}
    translation {z=2}
   }
   n20 = "Transform (Modify) Bone - Bone: base_L_skirt_a_1"!(ComponentPose=@n19, Alpha=@n16, Translation="0,0,0", Rotation="10,0,0") -> (Pose) {
    rotationMode=BMM_Additive
    rotationSpace=BCS_BoneSpace
    translationMode=BMM_Additive
    translationSpace=BCS_BoneSpace
    alphaBoolBlend {blendOption=Linear}
    alphaScaleBiasClamp {bInterpResult=true}
    boneToModify {boneName=base_L_skirt_a_1}
    rotation {pitch=10}
   }
   n21 = "Transform (Modify) Bone - Bone: base_C_skirt_a_1"!(ComponentPose=@n20, Alpha=@n16, Translation="0,0,0", Rotation="10,0,0") -> (Pose) {
    rotationMode=BMM_Additive
    rotationSpace=BCS_BoneSpace
    translationMode=BMM_Additive
    translationSpace=BCS_BoneSpace
    alphaBoolBlend {blendOption=Linear}
    alphaScaleBiasClamp {bInterpResult=true}
    boneToModify {boneName=base_C_skirt_a_1}
    rotation {pitch=10}
   }
   n22 = "Transform (Modify) Bone - Bone: base_R_skirt_a_1"!(ComponentPose=@n21, Alpha=@n16, Translation="0,0,0", Rotation="10,0,0") -> (Pose) {
    rotationMode=BMM_Additive
    rotationSpace=BCS_BoneSpace
    translationMode=BMM_Additive
    translationSpace=BCS_BoneSpace
    alphaBoolBlend {blendOption=Linear}
    alphaScaleBiasClamp {bInterpResult=true}
    boneToModify {boneName=base_R_skirt_a_1}
    rotation {pitch=10}
   }
   Save cached pose 'FixSprintCP'(Pose=Component To Local(ComponentPose=@n22)) {cacheName=FixSprintCP}
   n23 = Anim Dynamics - Bone: dyn_upperarm_r_sleeves_bk!(
    ComponentPose=Local To Component(LocalPose=Use cached pose 'RigidBodiesCP'())
    Alpha=1
    GravityOverride:Vector="0,0,-980"
    ExternalForce="0,0,0"
   ) -> (Pose) {
    angularDampingOverride=0.9999
    angularSpringConstant=18000
    bAngularSpring=true
    bOverrideAngularDamping=true
    bUseGravityOverride=true
    gravityScale=0.35
    simulationSpace=World
    alphaBoolBlend {blendOption=Linear}
    boundBone {boneName=dyn_upperarm_r_sleeves_bk}
    componentAppliedLinearAccClamp=<10000 10000 10000>
    componentLinearAccScale=<0.3 0.4 0.3>
    componentLinearVelScale=<0.3 0.4 0.3>
    gravityOverride {z=-980}
    physicsBodyDefinitions [{
      boundBone {boneName=dyn_upperarm_r_sleeves_bk}
      boxExtents=<25 25 25>
      constraintSetup {angularLimitsMax=<0 5 15>, angularLimitsMin=<0 -5 -5>, angularTarget=<1 0 0>}
      localJointOffset=<2 0 0>
     }]
   }
   n24 = Anim Dynamics - Bone: dyn_upperarm_r_sleeves_fw!(ComponentPose=@n23, Alpha=1, GravityOverride:Vector="0,0,-980", ExternalForce="0,0,0") -> (Pose) {
    angularDampingOverride=0.9999
    angularSpringConstant=18000
    bAngularSpring=true
    bOverrideAngularDamping=true
    bUseGravityOverride=true
    gravityScale=0.35
    simulationSpace=World
    alphaBoolBlend {blendOption=Linear}
    boundBone {boneName=dyn_upperarm_r_sleeves_fw}
    componentAppliedLinearAccClamp=<10000 10000 10000>
    componentLinearAccScale=<0.3 0.4 0.3>
    componentLinearVelScale=<0.3 0.4 0.3>
    gravityOverride {z=-980}
    physicsBodyDefinitions [{
      boundBone {boneName=dyn_upperarm_r_sleeves_fw}
      boxExtents=<25 25 25>
      constraintSetup {angularLimitsMax=<0 5 5>, angularLimitsMin=<0 -5 -15>, angularTarget=<1 0 0>}
      localJointOffset=<2 0 0>
     }]
   }
   n25 = Anim Dynamics - Bone: dyn_upperarm_r_sleeves_up!(ComponentPose=@n24, Alpha=1, GravityOverride:Vector="0,0,-980", ExternalForce="0,0,0") -> (Pose) {
    angularDampingOverride=0.9999
    angularSpringConstant=18000
    bAngularSpring=true
    bOverrideAngularDamping=true
    bUseGravityOverride=true
    gravityScale=0.35
    simulationSpace=World
    alphaBoolBlend {blendOption=Linear}
    boundBone {boneName=dyn_upperarm_r_sleeves_up}
    componentAppliedLinearAccClamp=<10000 10000 10000>
    componentLinearAccScale=<0.3 0.4 0.3>
    componentLinearVelScale=<0.3 0.4 0.3>
    gravityOverride {z=-980}
    physicsBodyDefinitions [{
      boundBone {boneName=dyn_upperarm_r_sleeves_up}
      boxExtents=<25 25 25>
      constraintSetup {angularLimitsMax=<0 5 5>, angularLimitsMin=<0 -15 -5>, angularTarget=<1 0 0>}
      localJointOffset=<2 0 0>
     }]
   }
   n26 = Anim Dynamics - Bone: dyn_upperarm_r_sleeves_dw!(ComponentPose=@n25, Alpha=1, GravityOverride:Vector="0,0,-980", ExternalForce="0,0,0") -> (Pose) {
    angularDampingOverride=0.9999
    angularSpringConstant=18000
    bAngularSpring=true
    bOverrideAngularDamping=true
    bUseGravityOverride=true
    gravityScale=0.35
    simulationSpace=World
    alphaBoolBlend {blendOption=Linear}
    boundBone {boneName=dyn_upperarm_r_sleeves_dw}
    componentAppliedLinearAccClamp=<10000 10000 10000>
    componentLinearAccScale=<0.3 0.4 0.3>
    componentLinearVelScale=<0.3 0.4 0.3>
    gravityOverride {z=-980}
    physicsBodyDefinitions [{
      boundBone {boneName=dyn_upperarm_r_sleeves_dw}
      boxExtents=<25 25 25>
      constraintSetup {angularLimitsMax=<0 10 5>, angularLimitsMin=<0 -10 -5>, angularTarget=<1 0 0>}
      localJointOffset=<2 0 0>
     }]
   }
   n27 = Anim Dynamics - Bone: dyn_upperarm_l_sleeves_bk!(ComponentPose=@n26, Alpha=1, GravityOverride:Vector="0,0,-980", ExternalForce="0,0,0") -> (Pose) {
    angularDampingOverride=0.9999
    angularSpringConstant=18000
    bAngularSpring=true
    bOverrideAngularDamping=true
    bUseGravityOverride=true
    gravityScale=0.35
    simulationSpace=World
    alphaBoolBlend {blendOption=Linear}
    boundBone {boneName=dyn_upperarm_l_sleeves_bk}
    componentAppliedLinearAccClamp=<10000 10000 10000>
    componentLinearAccScale=<0.3 0.4 0.3>
    componentLinearVelScale=<0.3 0.4 0.3>
    gravityOverride {z=-980}
    physicsBodyDefinitions [{
      boundBone {boneName=dyn_upperarm_l_sleeves_bk}
      boxExtents=<25 25 25>
      constraintSetup {angularLimitsMax=<0 5 5>, angularLimitsMin=<0 -5 -15>, angularTarget=<1 0 0>}
      localJointOffset=<2 0 0>
     }]
   }
   n28 = Anim Dynamics - Bone: dyn_upperarm_l_sleeves_fw!(ComponentPose=@n27, Alpha=1, GravityOverride:Vector="0,0,-980", ExternalForce="0,0,0") -> (Pose) {
    angularDampingOverride=0.9999
    angularSpringConstant=18000
    bAngularSpring=true
    bOverrideAngularDamping=true
    bUseGravityOverride=true
    gravityScale=0.35
    simulationSpace=World
    alphaBoolBlend {blendOption=Linear}
    boundBone {boneName=dyn_upperarm_l_sleeves_fw}
    componentAppliedLinearAccClamp=<10000 10000 10000>
    componentLinearAccScale=<0.3 0.4 0.3>
    componentLinearVelScale=<0.3 0.4 0.3>
    gravityOverride {z=-980}
    physicsBodyDefinitions [{
      boundBone {boneName=dyn_upperarm_l_sleeves_fw}
      boxExtents=<25 25 25>
      constraintSetup {angularLimitsMax=<0 5 15>, angularLimitsMin=<0 -5 -5>, angularTarget=<1 0 0>}
      localJointOffset=<2 0 0>
     }]
   }
   n29 = Anim Dynamics - Bone: dyn_upperarm_l_sleeves_up!(ComponentPose=@n28, Alpha=1, GravityOverride:Vector="0,0,-980", ExternalForce="0,0,0") -> (Pose) {
    angularDampingOverride=0.9999
    angularSpringConstant=18000
    bAngularSpring=true
    bOverrideAngularDamping=true
    bUseGravityOverride=true
    gravityScale=0.35
    simulationSpace=World
    alphaBoolBlend {blendOption=Linear}
    boundBone {boneName=dyn_upperarm_l_sleeves_up}
    componentAppliedLinearAccClamp=<10000 10000 10000>
    componentLinearAccScale=<0.3 0.4 0.3>
    componentLinearVelScale=<0.3 0.4 0.3>
    gravityOverride {z=-980}
    physicsBodyDefinitions [{
      boundBone {boneName=dyn_upperarm_l_sleeves_up}
      boxExtents=<25 25 25>
      constraintSetup {angularLimitsMax=<0 5 5>, angularLimitsMin=<0 -15 -5>, angularTarget=<1 0 0>}
      localJointOffset=<2 0 0>
     }]
   }
   n30 = Anim Dynamics - Bone: dyn_upperarm_l_sleeves_dw!(ComponentPose=@n29, Alpha=1, GravityOverride:Vector="0,0,-980", ExternalForce="0,0,0") -> (Pose) {
    angularDampingOverride=0.9999
    angularSpringConstant=18000
    bAngularSpring=true
    bOverrideAngularDamping=true
    bUseGravityOverride=true
    gravityScale=0.35
    simulationSpace=World
    alphaBoolBlend {blendOption=Linear}
    boundBone {boneName=dyn_upperarm_l_sleeves_dw}
    componentAppliedLinearAccClamp=<10000 10000 10000>
    componentLinearAccScale=<0.3 0.4 0.3>
    componentLinearVelScale=<0.3 0.4 0.3>
    gravityOverride {z=-980}
    physicsBodyDefinitions [{
      boundBone {boneName=dyn_upperarm_l_sleeves_dw}
      boxExtents=<25 25 25>
      constraintSetup {angularLimitsMax=<0 10 5>, angularLimitsMin=<0 -10 -5>, angularTarget=<1 0 0>}
      localJointOffset=<2 0 0>
     }]
   }
   Save cached pose 'animDynamicsCP'(Pose=Component To Local(ComponentPose=@n30)) {cacheName=animDynamicsCP}
   n31 = Get bIsSkydiveFloating()
   n32 = "Transform (Modify) Bone - Bone: dyn_L_skirt_c_3"!(
    ComponentPose=Local To Component(LocalPose=Use cached pose 'animDynamicsCP'())
    bAlphaBoolEnabled:bool=@n31
    Translation:Vector="0,0,0"
    Rotation="30,0,0"
   ) -> (Pose) {
    alphaInputType=Bool
    rotationMode=BMM_Additive
    rotationSpace=BCS_BoneSpace
    alphaBoolBlend {blendInTime=2, blendOption=Cubic, blendOutTime=1}
    boneToModify {boneName=dyn_L_skirt_c_3}
    rotation {pitch=30}
   }
   n33 = "Transform (Modify) Bone - Bone: dyn_R_skirt_c_3"!(ComponentPose=@n32, bAlphaBoolEnabled:bool=@n31, Translation:Vector="0,0,0", Rotation="30,0,0") -> (Pose) {
    alphaInputType=Bool
    rotationMode=BMM_Additive
    rotationSpace=BCS_BoneSpace
    alphaBoolBlend {blendInTime=2, blendOption=Cubic, blendOutTime=1}
    boneToModify {boneName=dyn_R_skirt_c_3}
    rotation {pitch=30}
   }
   n34 = "Transform (Modify) Bone - Bone: dyn_C_skirt_b_3"!(ComponentPose=@n33, bAlphaBoolEnabled:bool=@n31, Translation:Vector="0,0,0", Rotation="30,0,0") -> (Pose) {
    alphaInputType=Bool
    rotationMode=BMM_Additive
    rotationSpace=BCS_BoneSpace
    alphaBoolBlend {blendInTime=2, blendOption=Cubic, blendOutTime=1}
    boneToModify {boneName=dyn_C_skirt_b_3}
    rotation {pitch=30}
   }
   Save cached pose 'FixSkydiveCP'(Pose=Component To Local(ComponentPose=@n34)) {cacheName=FixSkydiveCP}
   n35 = Get isSkyDiving float() -> (isSkyDiving float:double)
   n36 = "Transform (Modify) Bone - Bone: attach_cape"(
    ComponentPose=Local To Component(LocalPose=Use cached pose 'fixRotateBackpackCP'())
    Alpha=@n35
    Translation="0,-4.5,0"
    Rotation="0,0,10"
    Scale="1,1,1"
   ) {
    rotationMode=BMM_Additive
    rotationSpace=BCS_BoneSpace
    translationMode=BMM_Additive
    translationSpace=BCS_BoneSpace
    alphaBoolBlend {blendInTime=0.1, blendOption=Linear, blendOutTime=0.1}
    alphaScaleBiasClamp {bInterpResult=true, interpSpeedDecreasing=2}
    boneToModify {boneName=attach_cape}
    rotation {roll=10}
    translation {y=-4.5}
   }
   n37 = "Transform (Modify) Bone - Bone: attach_backpack"(ComponentPose=@n36, Alpha=@n35, Translation="0,-6,0", Rotation="0,0,8", Scale="1,1,1") {
    rotationMode=BMM_Additive
    rotationSpace=BCS_BoneSpace
    translationMode=BMM_Additive
    translationSpace=BCS_BoneSpace
    alphaBoolBlend {blendInTime=0.1, blendOption=Linear, blendOutTime=0.1}
    alphaScaleBiasClamp {bInterpResult=true, interpSpeedDecreasing=2}
    boneToModify {boneName=attach_backpack}
    rotation {roll=8}
    translation {y=-6}
   }
   Save cached pose 'backPackNcapeMoveCP'(Pose=Component To Local(ComponentPose=@n37)) {cacheName=backPackNcapeMoveCP}
   n38 = Get bIsTacticalSprinting()
   n39 = "Constraint - dyn_upperarm_l_sleeves_bk (P)"!(ComponentPose=Local To Component(LocalPose=Use cached pose 'backPackNcapeMoveCP'()), bAlphaBoolEnabled:bool=@n38, ConstraintWeights_0=1) -> (Pose) {
    alphaInputType=Bool
    alphaBoolBlend {blendInTime=1, blendOption=Cubic, blendOutTime=1}
    boneToModify {boneName=dyn_upperarm_l_sleeves_bk}
    constraintSetup [{targetBone {boneName=upperarm_twist_01_l}}]
    constraintWeights [1]
   }
   n40 = "Constraint - dyn_upperarm_r_sleeves_bk (P)"!(ComponentPose=@n39, bAlphaBoolEnabled:bool=@n38, ConstraintWeights_0=1) -> (Pose) {
    alphaInputType=Bool
    alphaBoolBlend {blendInTime=1, blendOption=Cubic, blendOutTime=1}
    boneToModify {boneName=dyn_upperarm_r_sleeves_bk}
    constraintSetup [{targetBone {boneName=upperarm_twist_01_r}}]
    constraintWeights [1]
   }
   n41 = "Transform (Modify) Bone - Bone: dyn_upperarm_l_sleeves_bk"!(ComponentPose=@n40, bAlphaBoolEnabled:bool=@n38, Translation="0,2,0", Rotation="0,0,0", Scale="1,1,1") -> (Pose) {
    alphaInputType=Bool
    translationMode=BMM_Additive
    translationSpace=BCS_BoneSpace
    alphaBoolBlend {blendInTime=1, blendOption=Cubic, blendOutTime=1}
    boneToModify {boneName=dyn_upperarm_l_sleeves_bk}
    translation {y=2}
   }
   n42 = "Transform (Modify) Bone - Bone: dyn_upperarm_r_sleeves_bk"!(ComponentPose=@n41, bAlphaBoolEnabled:bool=@n38, Translation="0,-2,0", Rotation="0,0,0", Scale="1,1,1") -> (Pose) {
    alphaInputType=Bool
    translationMode=BMM_Additive
    translationSpace=BCS_BoneSpace
    alphaBoolBlend {blendInTime=1, blendOption=Cubic, blendOutTime=1}
    boneToModify {boneName=dyn_upperarm_r_sleeves_bk}
    translation {y=-2}
   }
   Save cached pose 'FixSleevesOnSpritingCP'(Pose=Component To Local(ComponentPose=@n42)) {cacheName=FixSleevesOnSpritingCP}
   n43 = RigidBody(
    ComponentPose=Local To Component(LocalPose=Use cached pose 'FixSprintCP'())
    Alpha=Get alphaSkirtRigidBody float() -> (alphaSkirtRigidBody float:double)
    OverrideWorldGravity=Get GravityOverride()
   ) {
    bOverrideWorldGravity=true
    overridePhysicsAsset=&/F_MED_HayBrush_Physics
    simulationSpace=BaseBoneSpace
    alphaBoolBlend {blendOption=Linear}
    alphaScaleBiasClamp {bInterpResult=true, interpSpeedDecreasing=2, interpSpeedIncreasing=2}
    baseBoneRef {boneName=root}
    componentAppliedLinearAccClamp=<3000 3000 3000>
    componentLinearAccScale=<0.3 0.4 0.3>
    componentLinearVelScale=<0.3 0.4 0.3>
    overrideWorldGravity {z=-980}
    simSpaceSettings {
     maxAngularAcceleration=600
     maxAngularVelocity=1000
     maxLinearAcceleration=500
     maxLinearVelocity=1000
     worldAlpha=0.4
     externalLinearDragV=<0.2 0.2>
    }
   }
   n44 = RigidBody(ComponentPose=Local To Component(LocalPose=Use cached pose 'FixSprintCP'()), OverrideWorldGravity=Get GravityOverride()) {
    bOverrideWorldGravity=true
    overridePhysicsAsset=&/F_MED_HayBrush_Physics_frontEnd
    simulationSpace=BaseBoneSpace
    alphaBoolBlend {blendOption=Linear}
    alphaScaleBiasClamp {interpSpeedDecreasing=0.1, interpSpeedIncreasing=0.1}
    baseBoneRef {boneName=root}
    componentAppliedLinearAccClamp=<3000 3000 3000>
    componentLinearAccScale=<0.3 0.4 0.3>
    componentLinearVelScale=<0.3 0.4 0.3>
    overrideWorldGravity {z=-980}
    simSpaceSettings {
     maxAngularAcceleration=600
     maxAngularVelocity=1000
     maxLinearAcceleration=500
     maxLinearVelocity=1000
     worldAlpha=0.4
     externalLinearDragV=<0.2 0.2>
    }
   }
   Save cached pose 'RigidBodiesCP'(Pose=Blend Poses by bool(BlendPose_0=Component To Local(ComponentPose=@n44), BlendPose_1=Component To Local(ComponentPose=@n43), bActiveValue=Get bIsInFrontEnd()) {blendType=Linear}) {cacheName=RigidBodiesCP}
   n45 = Get moveBackward bool() -> (moveBackward bool:bool)
   n46 = Get moveRight bool() -> (moveRight bool:bool)
   n47 = Get moveLeft bool() -> (moveLeft bool:bool)
   n48 = Get moveForward bool() -> (moveForward bool:bool)
   n49 = "Transform (Modify) Bone - Bone: dyn_L_skirt_c_3"!(
    ComponentPose=Local To Component(LocalPose=Use cached pose 'FixSleevesOnSpritingCP'())
    bAlphaBoolEnabled=@n48
    Translation="0,0,5"
    Rotation="0,0,0"
    Scale:Vector="1,1,1"
   ) -> (Pose) {
    alphaInputType=Bool
    rotationMode=BMM_Additive
    rotationSpace=BCS_BoneSpace
    scaleMode=BMM_Additive
    scaleSpace=BCS_BoneSpace
    translationMode=BMM_Additive
    translationSpace=BCS_BoneSpace
    alphaBoolBlend {blendInTime=0.3, blendOption=CubicInOut, blendOutTime=0.3}
    boneToModify {boneName=dyn_L_skirt_c_3}
    translation {z=5}
   }
   n50 = "Transform (Modify) Bone - Bone: dyn_C_skirt_b_3"!(ComponentPose=@n49, bAlphaBoolEnabled=@n48, Translation="0,0,2.5", Rotation="0,0,0", Scale:Vector="1,1,1") -> (Pose) {
    alphaInputType=Bool
    rotationMode=BMM_Additive
    rotationSpace=BCS_BoneSpace
    scaleMode=BMM_Additive
    scaleSpace=BCS_BoneSpace
    translationMode=BMM_Additive
    translationSpace=BCS_BoneSpace
    alphaBoolBlend {blendInTime=0.3, blendOption=CubicInOut, blendOutTime=0.3}
    boneToModify {boneName=dyn_C_skirt_b_3}
    translation {z=2.5}
   }
   n51 = "Transform (Modify) Bone - Bone: dyn_R_skirt_c_3"!(ComponentPose=@n50, bAlphaBoolEnabled=@n48, Translation="0,0,5", Rotation="0,0,0", Scale:Vector="1,1,1") -> (Pose) {
    alphaInputType=Bool
    rotationMode=BMM_Additive
    rotationSpace=BCS_BoneSpace
    scaleMode=BMM_Additive
    scaleSpace=BCS_BoneSpace
    translationMode=BMM_Additive
    translationSpace=BCS_BoneSpace
    alphaBoolBlend {blendInTime=0.3, blendOption=CubicInOut, blendOutTime=0.3}
    boneToModify {boneName=dyn_R_skirt_c_3}
    translation {z=5}
   }
   n52 = "Transform (Modify) Bone - Bone: base_L_skirt_b_1"!(ComponentPose=@n51, bAlphaBoolEnabled:bool=@n47, Translation="0,0,5", Rotation="12,0,0", Scale="1,1,1") -> (Pose) {
    alphaInputType=Bool
    rotationMode=BMM_Additive
    rotationSpace=BCS_BoneSpace
    scaleMode=BMM_Additive
    scaleSpace=BCS_BoneSpace
    translationMode=BMM_Additive
    translationSpace=BCS_BoneSpace
    alphaBoolBlend {blendInTime=0.3, blendOption=CubicInOut, blendOutTime=0.3}
    boneToModify {boneName=base_L_skirt_b_1}
    rotation {pitch=12}
    translation {z=5}
   }
   n53 = "Transform (Modify) Bone - Bone: base_R_skirt_c_2"!(ComponentPose=@n52, bAlphaBoolEnabled:bool=@n47, Translation="0,0,3", Rotation="0,0,0", Scale="1,1,1") -> (Pose) {
    alphaInputType=Bool
    rotationMode=BMM_Additive
    rotationSpace=BCS_BoneSpace
    scaleMode=BMM_Additive
    scaleSpace=BCS_BoneSpace
    translationMode=BMM_Additive
    translationSpace=BCS_BoneSpace
    alphaBoolBlend {blendInTime=0.3, blendOption=CubicInOut, blendOutTime=0.3}
    boneToModify {boneName=base_R_skirt_c_2}
    translation {z=3}
   }
   n54 = "Transform (Modify) Bone - Bone: base_R_skirt_b_1"!(ComponentPose=@n53, bAlphaBoolEnabled:bool=@n47, Translation="0,0,0", Rotation="12,0,0", Scale="1,1,1") -> (Pose) {
    alphaInputType=Bool
    rotationMode=BMM_Additive
    rotationSpace=BCS_BoneSpace
    scaleMode=BMM_Additive
    scaleSpace=BCS_BoneSpace
    translationMode=BMM_Additive
    translationSpace=BCS_BoneSpace
    alphaBoolBlend {blendInTime=0.3, blendOption=CubicInOut, blendOutTime=0.3}
    boneToModify {boneName=base_R_skirt_b_1}
    rotation {pitch=12}
   }
   n55 = "Transform (Modify) Bone - Bone: base_L_skirt_b_1"(ComponentPose=@n54, bAlphaBoolEnabled=@n46, Translation="0,0,0", Rotation="12,0,0", Scale="1,1,1") {
    alphaInputType=Bool
    rotationMode=BMM_Additive
    rotationSpace=BCS_BoneSpace
    scaleMode=BMM_Additive
    scaleSpace=BCS_BoneSpace
    translationMode=BMM_Additive
    translationSpace=BCS_BoneSpace
    alphaBoolBlend {blendInTime=0.3, blendOption=CubicInOut, blendOutTime=0.3}
    boneToModify {boneName=base_L_skirt_b_1}
    rotation {pitch=12}
   }
   n56 = "Transform (Modify) Bone - Bone: base_R_skirt_c_2"(ComponentPose=@n55, bAlphaBoolEnabled=@n46, Translation="0,0,5", Rotation="0,0,0", Scale="1,1,1") {
    alphaInputType=Bool
    rotationMode=BMM_Additive
    rotationSpace=BCS_BoneSpace
    scaleMode=BMM_Additive
    scaleSpace=BCS_BoneSpace
    translationMode=BMM_Additive
    translationSpace=BCS_BoneSpace
    alphaBoolBlend {blendInTime=0.3, blendOption=CubicInOut, blendOutTime=0.3}
    boneToModify {boneName=base_R_skirt_c_2}
    translation {z=5}
   }
   n57 = "Transform (Modify) Bone - Bone: base_R_skirt_b_1"(ComponentPose=@n56, bAlphaBoolEnabled=@n46, Translation="0,0,5", Rotation="15,0,0", Scale="1,1,1") {
    alphaInputType=Bool
    rotationMode=BMM_Additive
    rotationSpace=BCS_BoneSpace
    scaleMode=BMM_Additive
    scaleSpace=BCS_BoneSpace
    translationMode=BMM_Additive
    translationSpace=BCS_BoneSpace
    alphaBoolBlend {blendInTime=0.3, blendOption=CubicInOut, blendOutTime=0.3}
    boneToModify {boneName=base_R_skirt_b_1}
    rotation {pitch=15}
    translation {z=5}
   }
   n58 = "Transform (Modify) Bone - Bone: base_L_skirt_c_2"!(ComponentPose=@n57, bAlphaBoolEnabled:bool=@n45, Translation="0,0,3", Rotation="10,0,0", Scale="1,1,1") -> (Pose) {
    alphaInputType=Bool
    rotationMode=BMM_Additive
    rotationSpace=BCS_BoneSpace
    scaleMode=BMM_Additive
    scaleSpace=BCS_BoneSpace
    translationMode=BMM_Additive
    translationSpace=BCS_BoneSpace
    alphaBoolBlend {blendInTime=0.3, blendOption=CubicInOut, blendOutTime=0.3}
    boneToModify {boneName=base_L_skirt_c_2}
    rotation {pitch=10}
    translation {z=3}
   }
   n59 = "Transform (Modify) Bone - Bone: base_C_skirt_b_2"!(ComponentPose=@n58, bAlphaBoolEnabled:bool=@n45, Translation="0,0,3", Rotation="10,0,0", Scale="1,1,1") -> (Pose) {
    alphaInputType=Bool
    rotationMode=BMM_Additive
    rotationSpace=BCS_BoneSpace
    scaleMode=BMM_Additive
    scaleSpace=BCS_BoneSpace
    translationMode=BMM_Additive
    translationSpace=BCS_BoneSpace
    alphaBoolBlend {blendInTime=0.3, blendOption=CubicInOut, blendOutTime=0.3}
    boneToModify {boneName=base_C_skirt_b_2}
    rotation {pitch=10}
    translation {z=3}
   }
   n60 = "Transform (Modify) Bone - Bone: base_R_skirt_c_2"(ComponentPose=@n59, bAlphaBoolEnabled=@n45, Translation="0,0,3", Rotation="10,0,0", Scale="1,1,1") {
    alphaInputType=Bool
    rotationMode=BMM_Additive
    rotationSpace=BCS_BoneSpace
    scaleMode=BMM_Additive
    scaleSpace=BCS_BoneSpace
    translationMode=BMM_Additive
    translationSpace=BCS_BoneSpace
    alphaBoolBlend {blendInTime=0.3, blendOption=CubicInOut, blendOutTime=0.3}
    boneToModify {boneName=base_R_skirt_c_2}
    rotation {pitch=10}
    translation {z=3}
   }
   n61 = "Transform (Modify) Bone - Bone: base_L_skirt_c_2"(ComponentPose=@n60, bAlphaBoolEnabled=True, Translation="0,0,0", Rotation="0,0,0", Scale="0.5,1,1") {
    alphaInputType=Bool
    rotationMode=BMM_Additive
    rotationSpace=BCS_BoneSpace
    scaleMode=BMM_Additive
    scaleSpace=BCS_BoneSpace
    translationMode=BMM_Additive
    translationSpace=BCS_BoneSpace
    alphaBoolBlend {blendInTime=0.3, blendOption=CubicInOut, blendOutTime=0.3}
    boneToModify {boneName=base_L_skirt_c_2}
    scale {x=0.5}
   }
   n62 = "Transform (Modify) Bone - Bone: base_C_skirt_b_2"(ComponentPose=@n61, bAlphaBoolEnabled=True, Translation="0,0,0", Rotation="0,0,0", Scale="0.5,1,1") {
    alphaInputType=Bool
    rotationMode=BMM_Additive
    rotationSpace=BCS_BoneSpace
    scaleMode=BMM_Additive
    scaleSpace=BCS_BoneSpace
    translationMode=BMM_Additive
    translationSpace=BCS_BoneSpace
    alphaBoolBlend {blendInTime=0.3, blendOption=CubicInOut, blendOutTime=0.3}
    boneToModify {boneName=base_C_skirt_b_2}
    scale {x=0.5}
   }
   n63 = "Transform (Modify) Bone - Bone: base_R_skirt_c_2"(ComponentPose=@n62, bAlphaBoolEnabled=True, Translation="0,0,0", Rotation="10,0,0", Scale="0.5,1,1") {
    alphaInputType=Bool
    rotationMode=BMM_Additive
    rotationSpace=BCS_BoneSpace
    scaleMode=BMM_Additive
    scaleSpace=BCS_BoneSpace
    translationMode=BMM_Additive
    translationSpace=BCS_BoneSpace
    alphaBoolBlend {blendInTime=0.3, blendOption=CubicInOut, blendOutTime=0.3}
    boneToModify {boneName=base_R_skirt_c_2}
    rotation {pitch=10}
    scale {x=0.5}
   }
   Save cached pose 'DBNOSkirtAdjustmentsCP'(Pose=Blend Poses by bool(BlendPose_0=Component To Local(ComponentPose=@n63), BlendPose_1=Use cached pose 'FixSleevesOnSpritingCP'(), bActiveValue=Get bIsDBNO()) {blendType=Linear}) {cacheName=DBNOSkirtAdjustmentsCP}
   n64 = "Transform (Modify) Bone - Bone: attach_backpack"!(
    ComponentPose=Local To Component(LocalPose=Use cached pose 'FixSkydiveCP'())
    bAlphaBoolEnabled:bool=True
    Translation="0,-0.25,0"
    Rotation="0,0,6"
   ) -> (Pose) {
    alphaInputType=Bool
    rotationMode=BMM_Additive
    rotationSpace=BCS_BoneSpace
    translationMode=BMM_Additive
    translationSpace=BCS_BoneSpace
    alphaBoolBlend {blendInTime=2, blendOption=Cubic, blendOutTime=1}
    boneToModify {boneName=attach_backpack}
    rotation {roll=6}
    translation {y=-0.25}
   }
   n65 = "Transform (Modify) Bone - Bone: attach_backpack"(ComponentPose=@n64, bAlphaBoolEnabled=Get bIsParachuteOpened(), Translation="0,2.5,0", Rotation="0,0,-2") {
    alphaInputType=Bool
    rotationMode=BMM_Additive
    rotationSpace=BCS_BoneSpace
    translationMode=BMM_Additive
    translationSpace=BCS_BoneSpace
    alphaBoolBlend {blendInTime=5, blendOption=Linear, blendOutTime=1}
    alphaScaleBiasClamp {interpSpeedDecreasing=1, interpSpeedIncreasing=1}
    boneToModify {boneName=attach_backpack}
    rotation {roll=-2}
    translation {y=2.5}
   }
   Save cached pose 'fixRotateBackpackCP'(Pose=Component To Local(ComponentPose=@n65)) {cacheName=fixRotateBackpackCP}
  ]}
 EventGraph {nodes [
   Event BlueprintUpdateAnimation()
   > n1 = Sequence!(execute) -> (then_0, then_1, then_2:Exec, then_3:Exec)
   n2 = Sequence(execute=@n1.then_1)
   > n3 = Branch(Condition=OR Boolean(A=Get bIsSkydiveDiveMode(), B=Get bIsSkydiveFloating()))
   > Set enableSin(enableSin:bool=true) -> (Output_Get:bool)
   Set enableSin(execute=@n3.else, enableSin=false)
   n4 = Branch(execute=@n2.then_1, Condition=Get bIsSkydiveDiveMode())
   > Set freqVar1(freqVar1:double=40) -> (Output_Get:double)
   > Set freqVar2(freqVar2:double=38) -> (Output_Get:double)
   > Set freqVar3(freqVar3:double=35) -> (Output_Get:double)
   Set freqVar1(execute=@n4.else, freqVar1=35)
   > Set freqVar2(freqVar2=32)
   > Set freqVar3(freqVar3=30)
   Set sleeveRotVar1(
    execute=@n2.then_2
    sleeveRotVar1:Rotator=Rotator Rumble!(frequency:double=Get freqVar1() -> (freqVar1:double), amplitude:double=2, time offset:Vector="0,0,2", coefficient:Vector="0,0,1") -> (rotatorOutput:Rotator=~) {is=title K2Node_MacroInstance}
   ) -> (Output_Get:Rotator)
   > Set sleeveRotVar2(sleeveRotVar2:Rotator=Rotator Rumble(frequency=Get freqVar2() -> (freqVar2:double), amplitude=2, time offset="0,1,0", coefficient="0,0,1") -> (rotatorOutput=~)) -> (Output_Get:Rotator)
   Set skirtRotVar1(
    execute=@n2.then_3
    skirtRotVar1:Rotator=Rotator Rumble(frequency=Get freqVar1(), amplitude=2, time offset="0,1,2", coefficient="0,1,0") -> (rotatorOutput=~)
   ) -> (Output_Get:Rotator)
   > Set skirtRotVar2(skirtRotVar2:Rotator=Rotator Rumble(frequency=Get freqVar2(), amplitude=1.85, coefficient="0,1,0") -> (rotatorOutput=~)) -> (Output_Get:Rotator)
   > Set skirtRotVar3(skirtRotVar3:Rotator=Rotator Rumble(frequency=Get freqVar3() -> (freqVar3:double), amplitude=1.7, time offset="0,2,0", coefficient="0,1,0") -> (rotatorOutput=~)) -> (Output_Get:Rotator)
   n5 = Sequence!(execute=@n1.then_3) -> (then_0, then_1, then_2, then_3, then_4:Exec)
   > n6 = Branch(Condition=Get bIsKneeSliding())
   > Set isKneeSlidling float(isKneeSlidling float:double=1) -> (Output_Get:double)
   Set isKneeSlidling float(execute=@n6.else, isKneeSlidling float=0)
   n7 = Sequence!(execute=@n5.then_4) -> (then_0, then_1, then_2)
   n8 = Get bIsTacticalSprinting()
   n9 = Get bIsFalling()
   n10 = Get bIsJumping()
   > n11 = Branch(Condition=OR Boolean!(A=AND Boolean(A=@n10, B=@n8), B=AND Boolean(A=@n8, B=@n9), C:bool=Get bIsTacticalSprinting()) -> (ReturnValue))
   > Set alphaSkirtRigidBody float(alphaSkirtRigidBody float:double=0.5) -> (Output_Get:double)
   n12 = Branch(execute=@n11.else, Condition=Get bIsBaseballSliding())
   > Set alphaSkirtRigidBody float(alphaSkirtRigidBody float=0)
   n13 = Branch(execute=@n12.else, Condition=OR Boolean!(A=Get bIsJumping(), B=Get bIsFalling()) -> (ReturnValue))
   > Set alphaSkirtRigidBody float(alphaSkirtRigidBody float=0.15)
   Set alphaSkirtRigidBody float(execute=@n13.else, alphaSkirtRigidBody float=1)
   n14 = GetAnimInstance(self=GetOwnerSkeletalMeshComponent())
   Is Valid(exec=@n7.then_1, InputObject=@n14)
   > Set isPlayingEasterBloc(isPlayingEasterBloc:bool=Montage_IsPlaying(self=@n14, Montage=/BRCosmetics/Animation/Game/MainPlayer/Emotes/Eastern_Bloc/Emote_EasternBloc_CMF_M)) -> (Output_Get:bool)
   n15 = Branch(execute=@n5.then_1, Condition=Get bIsBaseballSliding())
   > Set isBaseballSlidling float(isBaseballSlidling float:double=1) -> (Output_Get:double)
   Set isBaseballSlidling float(execute=@n15.else, isBaseballSlidling float=0)
   n16 = Get moveLeft bool()
   n17 = Get bIsMoving2D()
   n18 = Get bIsCrouching()
   Set isCrouchStrafeLeft float(execute=@n5.then_3, isCrouchStrafeLeft float:double="To Float (Boolean)"(InBool=AND Boolean!(A=@n16, B=@n18, C:bool=@n17) -> (ReturnValue))) -> (Output_Get:double)
   n19 = Get moveRight bool()
   > Set isCrouchStrafeRight float(isCrouchStrafeRight float:double="To Float (Boolean)"(InBool=AND Boolean(A=@n18, B=@n17, C=@n19))) -> (Output_Get:double)
   n20 = PlayerMovement(execute=@n5.then_2, AnimInstance=Self-Reference(), XDotProductCoefficient=100, YDotProductCoefficient=100, ZDotProductCoefficient=100)
   n21 = BreakVector(InVec=MakeVector(X=@n20, Y=@n20.YMovement))
   > Set moveLeft bool(moveLeft bool:bool="float < float"(A=@n21, B=-50)) -> (Output_Get:bool)
   > Set moveRight bool(moveRight bool:bool="float > float"(A=@n21, B=50)) -> (Output_Get:bool)
   > Set moveForward bool(moveForward bool:bool="float > float"(A=@n21.Y, B=40)) -> (Output_Get:bool)
   > Set moveBackward bool(moveBackward bool:bool="float < float"(A=@n21.Y, B=-40)) -> (Output_Get:bool)
   n22 = Get bIsFalling()
   n23 = Branch(execute=@n1.then_2, Condition=AND Boolean!(A=Get bIsTacticalSprinting(), B=NOT Boolean(A=@n22)) -> (ReturnValue))
   > Set moveSkirtFwBk(moveSkirtFwBk:double=1) -> (Output_Get:double)
   n24 = Branch(execute=@n23.else, Condition=Get bIsSprinting())
   > Set moveSkirtFwBk(moveSkirtFwBk=0.2)
   Set moveSkirtFwBk(execute=@n24.else, moveSkirtFwBk=0)
   n25 = Gravity Override State Function DEPRECATED(
    execute=@n1.then_0
    CharacterPartAnimInstance=Self-Reference()
    GameTimeInSeconds=GetGameTimeInSeconds()
    GravityOverrideParams=Get Gravity Override Params()
   )
   > Set GravityOverride(GravityOverride=@n25)
   Event BlueprintBeginPlay() {is=title K2Node_Event}
   > Set disableControlRig(disableControlRig:bool=false) -> (Output_Get:bool)
   n26 = Event BlueprintProcessAnimInputEvent()
   > n27 = Branch(Condition=IsNamed(self=@n26.AnimInputEvent, InEventName=aie.Player.Zipline.Entered))
   > Set isZipline float(isZipline float:double=1) -> (Output_Get:double)
   Branch(execute=@n27.else, Condition=IsNamed(self=@n26.AnimInputEvent, InEventName=aie.Player.Zipline.Exited))
   > Set isZipline float(isZipline float=0)
   Event OnEnterGrindRail() {is=title K2Node_Event}
   > Set isOnGrindRail float(isOnGrindRail float:double=1) -> (Output_Get:double)
   Event OnEndSkydiving()
   > Set isSkyDiving float(isSkyDiving float:double=0) -> (Output_Get:double)
   Event OnBeginSkydiving()
   > Set isSkyDiving float(isSkyDiving float=1)
   Event OnExitGrindRail() {is=title K2Node_Event}
   > Set isOnGrindRail float(isOnGrindRail float=0)
   Delay(Duration=0.3)
  ]}
 floatSinFunc {nodes [
   n1 = Inputs!() -> (time:double=~, time offset:double=~, frequency:double=~, amplitude:double=~, coefficient:double=~)
   n2 = float + float(A=@n1, B=@n1.time offset)
   n3 = MultiplyByPi!(Value:double="Sin (Radians)"(A=float * float(A=@n2, B=@n1.frequency))) -> (ReturnValue:double)
   Outputs!(floatOutput:double=float * float(A=float * float(A=@n3, B=@n1.amplitude), B=@n1.coefficient))
  ]}
 rotatorRumble {nodes [
   n1 = Inputs!() -> (frequency, amplitude, time offset:Vector=~, coefficient:Vector=~)
   n2 = BreakVector(InVec=@n1.coefficient)
   n3 = BreakVector(InVec=@n1.time offset)
   n4 = GetGameTimeInSeconds()
   Outputs!(rotatorOutput:Rotator=MakeRotator(
     Roll=Float Sin Func!(time:double=@n4, time offset:double=@n3, frequency:double=@n1, amplitude:double=@n1.amplitude, coefficient:double=@n2) -> (floatOutput:double=~) {is=title K2Node_MacroInstance}
     Pitch=Float Sin Func(time=@n4, time offset=@n3.Y, frequency=@n1, amplitude=@n1.amplitude, coefficient=@n2.Y) -> (floatOutput=~)
     Yaw=Float Sin Func(time=@n4, time offset=@n3.Z, frequency=@n1, amplitude=@n1.amplitude, coefficient=@n2.Z) -> (floatOutput=~)
    ))
  ]}
]
variables [
 alpha sleeveDyn {type=number}
 alphaSkirtRigidBody float {type=number}
 disableControlRig {type=boolean}
 enableSin {type=boolean}
 freqVar1 {type=number}
 freqVar2 {type=number}
 freqVar3 {type=number}
 Gravity Override Params {type=GravityOverrideParamsStruct}
 GravityOverride {type=Vector}
 isBaseballSlidling float {type=number}
 isCrouchStrafeLeft float {type=number}
 isCrouchStrafeRight float {type=number}
 isKneeSlidling float {type=number}
 isOnGrindRail float {type=number}
 isPlayingEasterBloc {type=boolean}
 isSkyDiving float {type=number}
 isZipline float {type=number}
 moveBackward bool {type=boolean}
 moveForward bool {type=boolean}
 moveLeft bool {type=boolean}
 moveRight bool {type=boolean}
 moveSkirtFwBk {type=number}
 skirtRotVar1 {type=Rotator}
 skirtRotVar2 {type=Rotator}
 skirtRotVar3 {type=Rotator}
 sleeveRotVar1 {type=Rotator}
 sleeveRotVar2 {type=Rotator}
]
